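# Translucent, unlit, two-sided Niagara sprite material that draws a distorted FNCS logo for the SwordSplit emote.
# A tiling pattern, panned in radial coordinates and masked by an inverted soft sphere mask, offsets the logo UVs more toward the edges.
# TwoSidedSign flips U on back faces.
# The logo's red channel multiplies particle color, fed through an emissive adjustment function, and particle alpha.
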Material /BRCosmetics/Animation/Game/MainPlayer/Emotes/SwordSplit/FX/M_Emote_SwordSplit_FNCS_Logo
graph {
 properties {
  bAutomaticallySetUsageInEditor=false
  bUsedWithNiagaraSprites=true
  blendMode=BLEND_Translucent
  shadingModel=MSM_Unlit
  twoSided=true
  usedShadingModels=Unlit
 }
 nodes [
  LinearInterpolate()
  n1 = ParticleColor()
  n2 = MaterialFunctionCall(Vector or UVs=Add(A=TextureCoordinate()) {constB=-0.5}) {materialFunction=&VectorToRadialValue}
  n3 = TextureCoordinate()
  n4 = TextureSample!(
   UVs=Panner(Coordinate=Multiply(A=@n2.Radial Coordinates) {constB=2}) {speedX=-0.2, speedY=-0.792}
   Tex
   "DDX(UVs)"=DDX(Value=@n3)
   "DDY(UVs)"=DDY(Value=@n3)
   Apply View MipBias
  ) {mipValueMode=TMVM_Derivative, texture=/Game/Effects/Fort_Effects/Textures/T_Tile_Pattern_01}
  n5 = TextureCoordinate()
  n6 = Constant() {r=0.5}
  n7 = Constant()
  n8 = Multiply(
   A=Multiply(A=Power(Base=@n4.R) {constExponent=1.2}) {constB=0.06}
   B=OneMinus(None=SphereMask(A=@n5, B=@n6, Radius=@n6, Hardness=@n7) {hardnessPercent=0})
  )
  n9 = Multiply(A=Add(A=TextureCoordinate(), B=Saturate(None=@n8)), B=AppendVector(A=TwoSidedSign(), B=Constant() {r=1}))
  n10 = TextureSample!(UVs=@n9, Tex, Apply View MipBias) {samplerType=SAMPLERTYPE_LinearColor, texture=&/T_FNCS_Logo}
  MaterialFunctionCall(Emissive=Multiply(A=@n1.RGB, B=@n10.R)) {materialFunction=&MF_FX_EmissiveAdjustments}
  Multiply(A=@n1.A, B=@n10.R)
  TextureSample(UVs=@n9) {texture=&T_FNCS_Logo}
  TextureSample()
 ]
}
