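# Additive, unlit Niagara sprite material, apparently the sword glow for the SwordSplit emote.
# Particle color RGB feeds the shared emissive adjustment function.
# A stretched, vertically panning tile pattern is power-sharpened, shaped by a soft radial sphere mask and scaled by particle alpha to form an opacity value.
# A quality switch adds parameterized depth fade on High, Epic and default, skipping it on Low and Medium.
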Material /BRCosmetics/Animation/Game/MainPlayer/Emotes/SwordSplit/FX/M_Emote_SwordSplit_Sword_Glow
graph {
 properties {bUsedWithNiagaraSprites=true, blendMode=BLEND_Additive, pixelDepthOffsetMode=PDOM_Legacy, shadingModel=MSM_Unlit, usedShadingModels=Unlit}
 nodes [
  n1 = ParticleColor()
  MaterialFunctionCall(Emissive=@n1.RGB) {materialFunction=&MF_FX_EmissiveAdjustments}
  n2 = TextureCoordinate() {uTiling=0.22, vTiling=0.8}
  n3 = Power(Base=TextureSample(UVs=Panner(Coordinate=@n2) {speedY=0.5}) {texture=/Game/Effects/Fort_Effects/Textures/T_Tile_Pattern_01}.R) {constExponent=2.2}
  n4 = TextureCoordinate()
  n5 = Constant() {r=0.5}
  n6 = Constant()
  n7 = Multiply(A=@n1.A, B=Multiply(A=Saturate(None=@n3), B=SphereMask(A=@n4, B=@n5, Radius=@n5, Hardness=@n6)))
  n8 = DepthFade(Opacity=@n7, FadeDistance=ScalarParameter() {defaultValue=11, parameterName=Depth Fade})
  QualitySwitch(Default=@n8, Low=@n7, High=@n8, Medium=@n7, Epic=@n8)
 ]
}
