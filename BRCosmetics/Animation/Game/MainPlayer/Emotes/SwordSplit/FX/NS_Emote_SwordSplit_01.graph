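# Niagara system for the SwordSplit emote, layering orange glows, lightning, streaks, energy waves and flashes with lights on sockets or bones of a skeletal mesh.
# Timed one-shot bursts fire glow, wave, lightning and logo flash at 0.33 seconds, then connecting streaks and another flash around 1.1–1.2 seconds.
# Spawn groups vary particles within emitters, curves animate scale, color and material parameters, and brightness adapts to front-end, platform or time of day.
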NiagaraSystem /BRCosmetics/Animation/Game/MainPlayer/Emotes/SwordSplit/FX/NS_Emote_SwordSplit_01
dependencies {
 usedDataInterfaces [
  &NiagaraDataInterfaceCurve
  &NiagaraDataInterfaceColorCurve
  &NiagaraDataInterfacePlatformSet
  &NiagaraDataInterfaceArrayFloat
  &NiagaraDataInterfaceVector2DCurve
  &NiagaraDataInterfaceSkeletalMesh
 ]
 usedDynamicInputs []
 usedModules [
  &self:SystemSpawnScript.SystemScriptSource.SystemScriptGraph.NiagaraNodeAssignment_0.SetVariables_#1
  &SystemState
 ]
}
emitters {
 CapsuleGlow {
  emitterUpdateScript [
   &EmitterState(Inactive Response=::NewEnumerator1, Life Cycle Mode=::NewEnumerator1, Loop Behavior=::NewEnumerator1, Loop Duration=3, UseLoopDelay=false)
   &SpawnBurst_Instantaneous(Spawn Count=4, Spawn Time=0.33, Use Spawn Probability=false, UseLoopCountLimit=false)
  ]
  particleSpawnScript [
   &FN_SetSpawnGroup()
   &InitializeParticle(
    Color=&MergeHelper_ColorSwitch(
     0=<rgba 1 0.2384 0.0009 0.2>
     1=<rgba 1 0.2384 0.0009 0.2>
     2=<rgba 9 2.1453 0.0078 1>
     3=<rgba 9 2.1453 0.0078 1>
     Index=&SpawnGroup()
     Num=::NewEnumerator2
    )
    Lifetime=1.8
    Position Mode=::NewEnumerator0
    Sprite Rotation Mode=::NewEnumerator0
    Sprite Size=&MergeHelper_Vector2Switch(0=<22.1109 86>, 1=<22.1109 98>, 2=<5 94>, 3=<5 109>, 4=?, Index=&SpawnGroup(), Num=::NewEnumerator2)
    Sprite Size Mode=::NewEnumerator4
    Write Material Random=true
   )
   &CameraOffset(Camera Offset Amount=&MergeHelper_FloatSwitch(0=-10, 1=-10, 2=1, 3=1, Index=&SpawnGroup(), Num=::NewEnumerator2))
   /SetVariables(Particles.VisibilityTag=&MergeHelper_IntSwitch(0=0, 1=0, 2=1, 3=1, 4=?, 5=?, Index=&SpawnGroup(), Num=::NewEnumerator2)) {isSetParameters=true}
  ]
  particleUpdateScript [
   &ParticleState()
   &ScaleSpriteSize(
    Non-Uniform Curve Sprite Scale=DataInterface() {propertyValues {
      XCurve {keys [
        <at 0.3 0.43 RCIM_Cubic>
        <at 0.3482 0.43>
        {arriveTangent=2.4368, interpMode=RCIM_Cubic, leaveTangent=2.4368, time=0.4, value=1.1056}
        {tangentMode=RCTM_User, time=0.5, value=0.8}
       ]}
      YCurve {keys [
        <at 0.3 0.85 RCIM_Cubic>
        {arriveTangent=0.25, interpMode=RCIM_Cubic, leaveTangent=0.25, time=0.4, value=0.9649}
        {tangentMode=RCTM_User, time=0.5, value=0.9}
       ]}
     }}
    Scale Sprite Size Mode=::NewEnumerator3
   )
   &SkeletalMeshLocation(
    Bone / Socket Index=&MergeHelper_IntSwitch(0=0, 1=1, 2=0, 3=1, 4=?, 5=?, Index=&SpawnGroup(), Num=::NewEnumerator2)
    Bone Sampling Mode=::NewEnumerator4
    MaskSpawnGroup=false
    Sampled Position Offset=&MergeHelper_VectorSwitch(0=<0 0 46>, 1=<0 0 53>, 2=<0 -0.7 47>, 3=<0 0 53>, Index=&SpawnGroup(), Num=::NewEnumerator2)
    Skeletal Mesh=$System.Skeletal Mesh
   )
   &SpriteFacingAndAlignment(
    Sprite Alignment=&RotateVectorByQuaternion(Vector To Rotate=<0 0 1>, Quaternion=$Particles.SkeletalMeshLocation.SampledRotation)
    Sprite Facing=<0.7068 0.07 0.3>
    WriteAlignment=true
    WriteFacing=false
   )
   &ScaleColor(
    Scale Alpha=&FloatFromCurve(
     FloatCurve=DataInterface() {propertyValues {Curve {keys [
         <at 0 0 RCIM_Cubic>
         {arriveTangent=0.0972, interpMode=RCIM_Cubic, leaveTangent=0.0972, tangentMode=RCTM_User, time=0.05, value=1}
         {arriveTangent=-0.0885, interpMode=RCIM_Cubic, leaveTangent=-0.0885, tangentMode=RCTM_User, time=0.7, value=1}
         <at 1 0 RCIM_Cubic>
        ]}}}
     Scale Curve=&MergeHelper_FloatSwitch(
      0=&FloatFromCurve(FloatCurve=DataInterface() {propertyValues {Curve {keys [<at 0 0 RCIM_Constant>,<at 0.5 1 RCIM_Constant>]}}})
      1=1
      2=&FloatFromCurve(FloatCurve=DataInterface() {propertyValues {Curve {keys [<at 0 0 RCIM_Constant>,<at 0.5 1 RCIM_Constant>]}}})
      3=1
      Index=&SpawnGroup()
      Num=::NewEnumerator2
     )
    )
    Scale RGB=&VectorFromFloat(Value=&MergeHelper_FloatSwitch(
      0=1
      1=1
      2=1
      3=&FloatFromCurve(FloatCurve=DataInterface() {propertyValues {Curve {keys [
           {arriveTangent=0.1506, interpMode=RCIM_Cubic, leaveTangent=0.1506, tangentMode=RCTM_User, time=0, value=0.4}
           {arriveTangent=-0.0658, interpMode=RCIM_Cubic, leaveTangent=-0.0658, tangentMode=RCTM_User, time=0.2984, value=0.4}
           {arriveTangent=2.7544, interpMode=RCIM_Cubic, leaveTangent=2.7544, tangentMode=RCTM_User, time=0.3923, value=7}
           {arriveTangent=-1.6506, interpMode=RCIM_Cubic, leaveTangent=-1.6506, tangentMode=RCTM_User, time=0.5, value=1}
           <at 1 1>
          ]}}})
      Index=&SpawnGroup()
      Num=::NewEnumerator2
     ))
    ScaleA=true
    ScaleRGB=true
   )
  ]
  rendererClasses [&NiagaraSpriteRendererProperties]
  renderers [
   {rendererClass=&NiagaraSpriteRendererProperties}
   {rendererClass=&NiagaraSpriteRendererProperties, rendererIndex=1}
  ]
 }
 ConnectingStreaks {
  emitterUpdateScript [
   &EmitterState(Inactive Response=::NewEnumerator1, Life Cycle Mode=::NewEnumerator1, Loop Behavior=::NewEnumerator1, Loop Duration=3, UseLoopDelay=false)
   &SpawnBurst_Instantaneous(Spawn Count=7, Spawn Time=1.1, Use Spawn Probability=false, UseLoopCountLimit=false)
  ]
  particleSpawnScript [
   &FN_SetSpawnGroup(Number of Spawn Groups=2)
   /SetVariables(Particles.SizeToPosLerp:NiagaraFloat=&RandomRangeFloat(Minimum=0, Maximum=1)) {isSetParameters=true}
   /SetVariables(Particles.RandFloat:NiagaraFloat=&UniformAOrBFloat(
     A=&RandomRangeFloat(Minimum=0, Maximum=0.45)
     B=&RandomRangeFloat(Minimum=0.45, Maximum=1)
     Evaluation Type=::NewEnumerator0
     Recalculate Random Each Loop=true
     Override Seed=false
    )) {isSetParameters=true}
   &InitializeParticle(
    Color=<rgba 1 0.1802 0.0537 1>
    Lifetime Max=0.3
    Lifetime Min=0.2
    Lifetime Mode=::NewEnumerator1
    Mesh Scale=&Multiply_Vector(
     A=&RandomRangeVector(Minimum=<3 0.1 0.3>, Maximum=<3 0.1 0.5>)
     B=&VectorFromFloat(Value=&Lerp_Float(A=0.8, B=1.4, Alpha=$Particles.SizeToPosLerp))
    )
    Mesh Scale Mode=::NewEnumerator4
    UsePositionOffset=false
    Write Material Random=true
   )
   &CameraOffset(Camera Offset Amount=14)
   /SetVariables(Particles.Random=&RandomRangeFloat(Minimum=0, Maximum=1)) {isSetParameters=true, module=SetVariables_#2}
   &InitialMeshOrientation(
    Mesh Axis To Orient=&MergeHelper_VectorSwitch(0=<0 1 0>, 1=<0 1 0>, 2=<1 0 0>, 3=?, Index=&SpawnGroup(), Num=::NewEnumerator1)
    Mesh Orientation Mode=::NewEnumerator2
    Orientation Vector=&MergeHelper_VectorSwitch(0=<0 1 0>, 1=<0 -1 0>, 2=<1 0 0>, 3=?, Index=&SpawnGroup(), Num=::NewEnumerator1)
    Rotation=&RandomRangeVector(Minimum=<0 0 0>, Maximum=<1 0 0>)
    Use Rotation Vector=true
   )
  ]
  particleUpdateScript [
   &ParticleState()
   &ScaleMeshSize(Scale Factor=&Multiply_VectorByFloat(
     Vector=&MakeVector(
      X=1
      Y=&FloatFromCurve(FloatCurve=DataInterface() {propertyValues {Curve {keys [<at 0.2805 1 RCIM_Cubic>,<at 1 5>]}}})
      Z=1
     )
     Float=&FloatFromCurve(FloatCurve=DataInterface() {propertyValues {Curve {keys [<at 0 0 RCIM_Cubic>, {arriveTangent=2.7736, tangentMode=RCTM_User, time=1, value=1}]}}})
    ))
   &ScaleColor(
    Scale Alpha=&Multiply_Float(
     A=10
     B=&FloatFromCurve(
      FloatCurve=DataInterface() {propertyValues {Curve {keys [
          <at 0 0 RCIM_Cubic>
          {arriveTangent=-0.2922, interpMode=RCIM_Cubic, leaveTangent=-0.2922, tangentMode=RCTM_User, time=0.1, value=0}
          <at 0.2 0.8 RCIM_Cubic>
          {arriveTangent=-2.4195, interpMode=RCIM_Cubic, leaveTangent=-2.4195, tangentMode=RCTM_User, time=1, value=0}
         ]}}}
      Scale Curve=&FN_FrontEndFloatModifier(Gameplay=1, Front End=0.7)
     )
    )
    Scale RGB=&Multiply_VectorByFloat(
     Vector=<1 1 1>
     Float=&Multiply_Float(
      A=&RandomRangeFloat(Minimum=0, Maximum=1, Evaluation Type=::NewEnumerator1)
      B=&Multiply_Float(
       A=1
       B=&FloatFromCurve(FloatCurve=DataInterface() {propertyValues {Curve {keys [<at 0 10>,<at 1 10>]}}})
      )
     )
    )
    ScaleA=true
    ScaleRGB=true
   )
   &SkeletalMeshLocation(
    Bone / Socket Index=&MergeHelper_IntSwitch(
     0=0
     1=1
     2=&RandomRangeInt(Minimum=0, Maximum=1)
     3=?
     4=?
     5=?
     Index=&SpawnGroup()
     Num=::NewEnumerator1
    )
    Bone Sampling Mode=::NewEnumerator4
    MaskSpawnGroup=false
    Orientation Sampling=::NewEnumerator1
    Sampled Position Offset=&Lerp_Vector(A=<0 0 20>, B=<1 1 80>, Alpha=$Particles.SizeToPosLerp)
    Skeletal Mesh=$System.Skeletal Mesh
   )
   &DynamicMaterialParameters(
    Index 0 Param 1=&FloatFromCurve(FloatCurve=DataInterface() {propertyValues {Curve {keys [{tangentMode=RCTM_User, time=0, value=1}, {tangentMode=RCTM_User, time=1, value=1}]}}})
    Index 0 Param 2=0.7
    Index 0 Param 3=&FloatFromCurve(FloatCurve=DataInterface() {propertyValues {Curve {keys [
         {arriveTangent=12.4916, interpMode=RCIM_Cubic, leaveTangent=12.4916, tangentMode=RCTM_User, time=0, value=0}
         {arriveTangent=3, interpMode=RCIM_Cubic, leaveTangent=3, time=0.0944, value=1.0043}
         {arriveTangent=1.6993, interpMode=RCIM_Cubic, leaveTangent=1.6993, tangentMode=RCTM_User, time=1, value=3}
        ]}}})
    Index 0 Param 4=3
    Index 1 Param 1=&Multiply_Float(A=2, B=1)
    Index 1 Param 2=$Particles.Random
    Index 1 Param 3=0.7
    Index 1 Param 4=1
    Index 2 Param 1=&FloatFromCurve(FloatCurve=DataInterface() {propertyValues {Curve {keys [<at 0 1>,<at 1 0.3>]}}})
    Module.Param0WriteEnabled=true
    Module.Param0WriteEnabled1=true
    Module.Param0WriteEnabled2=true
    Module.Param1WriteEnabled=true
    Module.Param1WriteEnabled1=true
    Module.Param1WriteEnabled2=true
    Module.Param2WriteEnabled=true
    Module.Param2WriteEnabled1=true
    Module.Param2WriteEnabled2=true
    Module.Param3WriteEnabled=true
    Module.Param3WriteEnabled1=true
    Module.Param3WriteEnabled2=true
    Write Parameter Index 1=true
    Write Parameter Index 2=true
   )
   /SetVariables(Particles.Lightning glow color:LinearColor=&MakeLinearColorFromVectorAndFloat(
     "Vector (RGB)"=&MakeVectorFromLinearColorRGB(Color=&MultiplyLinearColor(
       Linear Color=<rgba 2 0.3115 0.0625 1>
       Scale Factor=&MakeVector4_FromFloat(Float=&RandomRangeFloat(Minimum=0.5, Maximum=1, Evaluation Type=::NewEnumerator1))
      ))
     "Float (Alpha)"=&Multiply_Float(
      A=&FloatFromCurve(FloatCurve=DataInterface() {propertyValues {Curve {keys [<at 0 1 RCIM_Cubic>, {arriveTangent=-3.4373, interpMode=RCIM_Cubic, leaveTangent=-3.4373, tangentMode=RCTM_User, time=0.77, value=0}]}}})
      B=2
     )
    )) {isSetParameters=true, module=SetVariables_#3}
   /SetVariables(Particles.Glow Dy02:Vector4f=&MakeVector4(
     X=&FloatFromCurve(FloatCurve=DataInterface() {propertyValues {Curve {keys [{tangentMode=RCTM_User, time=0, value=1}, {tangentMode=RCTM_User, time=0.5, value=1}]}}})
     Y=0.7
     Z=0
     W=3
    )) {isSetParameters=true, module=SetVariables_#4}
   /SetVariables(Particles.Glow Dy03:Vector4f=&MakeVector4(X=&Multiply_Float(A=1.3, B=1), Y=$Particles.Random, Z=0.7, W=1)) {isSetParameters=true, module=SetVariables_#5}
  ]
  rendererClasses [&NiagaraMeshRendererProperties]
  renderers [
   {rendererClass=&NiagaraMeshRendererProperties, rendererIndex=0}
   {rendererClass=&NiagaraMeshRendererProperties, rendererIndex=1}
  ]
 }
 EnergyWave {
  emitterUpdateScript [
   &EmitterState(Inactive Response=::NewEnumerator1, Life Cycle Mode=::NewEnumerator1, Loop Behavior=::NewEnumerator1, Loop Duration=2, UseLoopDelay=false)
   &SpawnBurst_Instantaneous(Spawn Time=0.33, Use Spawn Probability=false, UseLoopCountLimit=false)
   /SetVariables(
    Emitter.DepthFade:NiagaraFloat=&FN_FrontEndFloatModifier(Gameplay=500, Front End=300)
    Emitter.Refraction:NiagaraFloat=&FN_FrontEndFloatModifier(
     Gameplay=&FN_PickFloatDependOnPlatform(TargetPlatformActive=10, TargetPlatformNoActive=0)
     Front End=0
    )
   ) {isSetParameters=true}
  ]
  particleSpawnScript [
   &InitializeParticle(
    Color Mode=::NewEnumerator0
    Lifetime=0.3
    Mesh Scale Mode=::NewEnumerator3
    Mesh Uniform Scale=3
    UsePositionOffset=false
    Write Material Random=true
   )
   &CameraOffset()
   /SetVariables(Particles.LightPosOffset:NiagaraPosition=<10 10 0>) {isSetParameters=true}
   &InitialMeshOrientation(Mesh Axis To Orient=<0 0 1>, Mesh Orientation Mode=::NewEnumerator2, Use Rotation Vector=false)
  ]
  particleUpdateScript [
   &ParticleState()
   &ScaleColor(
    Linear Color Curve=DataInterface() {propertyValues {
      AlphaCurve {keys [<at 0 1>,<at 0.2388 1>,<at 1 0>]}
      BlueCurve {keys [<at 0 0>,<at 0.168 0>,<at 0.365 0.0728>,<at 1 0.0089>]}
      GreenCurve {keys [<at 0 0.2181>,<at 0.168 0.3824>,<at 0.365 0.395>,<at 1 0.1141>]}
      RedCurve {keys [<at 0 1>,<at 0.168 1>,<at 0.365 1>,<at 1 1>]}
     }}
    Scale Mode=::NewEnumerator2
   )
   &SkeletalMeshLocation(
    Bone Sampling Mode=::NewEnumerator4
    MaskSpawnGroup=false
    Orientation Sampling=::NewEnumerator1
    Sampled Position Offset=<0 0 28>
    Skeletal Mesh=$System.Skeletal Mesh
   )
   &DynamicMaterialParameters(
    Index 0 Param 1=&FloatFromCurve(FloatCurve=DataInterface() {propertyValues {Curve {keys [{arriveTangent=-2.1015, interpMode=RCIM_Cubic, leaveTangent=-2.1015, tangentMode=RCTM_User, time=0, value=0.5}, <at 1 -0.15>]}}})
    Index 0 Param 2=&FloatFromCurve(FloatCurve=DataInterface() {propertyValues {Curve {keys [{arriveTangent=-0.3464, interpMode=RCIM_Cubic, leaveTangent=-0.3464, tangentMode=RCTM_User, time=0, value=1}, <at 1 0.9>]}}})
    Index 0 Param 3=&FloatFromCurve(FloatCurve=DataInterface() {propertyValues {Curve {keys [
         {interpMode=RCIM_Cubic, tangentMode=RCTM_Break, time=0, value=0}
         {arriveTangent=0.4478, interpMode=RCIM_Cubic, leaveTangent=0.4478, tangentMode=RCTM_User, time=0.15, value=1}
         {interpMode=RCIM_Cubic, tangentMode=RCTM_Break, time=1, value=1.2}
        ]}}})
    Module.Param0WriteEnabled=true
    Module.Param1WriteEnabled=true
    Module.Param2WriteEnabled=true
    Module.Param3WriteEnabled=false
   )
   &ScaleColor(
    Scale Alpha=&FloatFromCurve(
     FloatCurve=DataInterface() {propertyValues {Curve {keys [{interpMode=RCIM_Cubic, tangentMode=RCTM_User, time=0.5, value=1}, {interpMode=RCIM_Cubic, tangentMode=RCTM_User, time=1, value=0}]}}}
     Scale Curve=&FN_FrontEndFloatModifier(Gameplay=0.8, Front End=0.5)
    )
    Scale RGB=&VectorFromFloat(Value=&FloatFromCurve(FloatCurve=DataInterface() {propertyValues {Curve {keys [
          {interpMode=RCIM_Cubic, tangentMode=RCTM_Break, time=0, value=10}
          {arriveTangent=-6.6072, interpMode=RCIM_Cubic, leaveTangent=-6.6072, tangentMode=RCTM_User, time=0.15, value=3}
          {interpMode=RCIM_Cubic, tangentMode=RCTM_Break, time=1, value=1}
         ]}}}))
    ScaleA=true
    ScaleRGB=true
   ) {module=ScaleColor001}
   &ScaleMeshSize(Scale Factor=&VectorFromFloat(Value=&FloatFromCurve(
      FloatCurve=DataInterface() {propertyValues {Curve {keys [
          {interpMode=RCIM_Cubic, tangentMode=RCTM_Break, time=0, value=0.2}
          {arriveTangent=2.9365, interpMode=RCIM_Cubic, leaveTangent=2.9365, tangentMode=RCTM_User, time=0.1, value=1}
          {interpMode=RCIM_Cubic, tangentMode=RCTM_Break, time=1, value=2}
         ]}}}
      Scale Curve=0.25
     )))
   &FN_Light_Attributes(
    Color=<rgba 3 0.5742 0.044 1>
    Create Color Attribute=true
    Create Exponent Attribute=true
    Create Position Attribute=true
    Create Radius Attribute=true
    Exponent=1.3
    Position=&Add_Vector(A=Linked() {linkedVariable {name=Particles.Initial.Position, type=NiagaraPosition}}, B=<0 0 0>)
    Radius=130
    Scale Color=&FloatFromCurve(
     FloatCurve=DataInterface() {propertyValues {Curve {keys [<at 0 1 RCIM_Cubic>,<at 1 0 RCIM_Cubic>]}}}
     Scale Curve=&Multiply_Float(A=15, B=1)
    )
   )
   /SetVariables(Particles.LightPos:NiagaraPosition=&Add_Vector(
     A=Linked() {linkedVariable {name=Particles.Position, type=NiagaraPosition}}
     B=Linked() {linkedVariable {name=Particles.LightPosOffset, type=NiagaraPosition}}
    )) {isSetParameters=true}
  ]
  rendererClasses [&NiagaraMeshRendererProperties, &NiagaraLightRendererProperties]
  renderers [
   {rendererClass=&NiagaraMeshRendererProperties, rendererIndex=0}
   {rendererClass=&NiagaraLightRendererProperties, rendererIndex=1}
  ]
 }
 Flash {
  emitterUpdateScript [
   &EmitterState(Inactive Response=::NewEnumerator1, Life Cycle Mode=::NewEnumerator1, Loop Behavior=::NewEnumerator1, Loop Duration=2, UseLoopDelay=false)
   &SpawnBurst_Instantaneous(Spawn Time=1.21, Use Spawn Probability=false, UseLoopCountLimit=false)
  ]
  particleSpawnScript [
   &InitializeParticle(
    Lifetime=0.24
    Mass Mode=::NewEnumerator3
    Mesh Scale Mode=::NewEnumerator3
    Mesh Uniform Scale=2.5
    Sprite Size Mode=::NewEnumerator3
    Uniform Sprite Size=350
    UsePositionOffset=true
    Write Material Random=true
   )
   &ShapeLocation(MaskSpawnGroup=false, Non Uniform Scale=<0.5 1 1>, Override Seed=false, Sphere Radius=0, UseOwnerScale=true)
   /SetVariables(Particles.SpriteAlignment=<1 0 0.4>) {isSetParameters=true}
   &InitialMeshOrientation(Orientation Axis=<0 0 1>, Orientation Vector=<0 0 1>, Rotation=<0 0.25 0>, Use Orientation Vector=true, Use Rotation Vector=true)
   &InitialMeshOrientation(
    Orientation Axis=<0 0 0>
    Orientation Vector=<0 0 0>
    Rotation=&RandomRangeVector(Minimum=<0 0 0>, Maximum=<1 0 0>)
    Use Orientation Vector=false
    Use Rotation Vector=true
   ) {module=InitialMeshOrientation001}
   &SkeletalMeshLocation(
    Bone / Socket Index=2
    Bone Sampling Mode=::NewEnumerator4
    MaskSpawnGroup=false
    Orientation Sampling=::NewEnumerator1
    Sampled Position Offset=<0 45 0>
    Skeletal Mesh=$System.Skeletal Mesh
   )
  ]
  particleUpdateScript [
   &ParticleState()
   &SpriteRotationRate(Rotation Rate=&FloatFromCurve(
     FloatCurve=DataInterface() {propertyValues {Curve {keys [
         {arriveTangent=1.4306, interpMode=RCIM_Cubic, leaveTangent=1.4306, tangentMode=RCTM_User, time=0.2, value=0}
         {arriveTangent=1.1179, interpMode=RCIM_Cubic, leaveTangent=1.1179, tangentMode=RCTM_User, time=0.4, value=0.6228}
         {arriveTangent=0.0224, interpMode=RCIM_Cubic, leaveTangent=0.0224, tangentMode=RCTM_User, time=1, value=1}
        ]}}}
     Scale Curve=0
    ))
   &ScaleColor(
    Scale Alpha=&FloatFromCurve(FloatCurve=DataInterface() {propertyValues {Curve {keys [
         {arriveTangent=-1.5276, interpMode=RCIM_Cubic, leaveTangent=-1.5276, tangentMode=RCTM_User, time=0, value=1}
         {interpMode=RCIM_Cubic, tangentMode=RCTM_User, time=1, value=0}
        ]}}})
    Scale RGB=&Multiply_Vector(
     A=&MakeVectorFromLinearColorRGB(Color=&ColorFromCurve(ColorCurve=DataInterface() {propertyValues {
         AlphaCurve {keys [<at 0 1 RCIM_Cubic>]}
         BlueCurve {keys [<at 0 20 RCIM_Cubic>,<at 0.6348 8>]}
         GreenCurve {keys [<at 0 20 RCIM_Cubic>,<at 0.6348 8>]}
         RedCurve {keys [<at 0 20 RCIM_Cubic>,<at 0.6348 8>]}
        }}))
     B=&MakeVectorFromLinearColorRGB(Color=<rgba 0.2552 0.0535 0.0023 1>)
    )
    ScaleA=true
    ScaleRGB=true
   ) {module=ScaleColor001}
   &ScaleSpriteSize(
    Scale Sprite Size Mode=::NewEnumerator0
    Uniform Scale Factor=&FloatFromCurve(FloatCurve=DataInterface() {propertyValues {Curve {keys [
         {arriveTangent=2.278, interpMode=RCIM_Cubic, leaveTangent=2.278, tangentMode=RCTM_User, time=0, value=0}
         {interpMode=RCIM_Cubic, tangentMode=RCTM_User, time=1, value=1}
        ]}}})
   )
   /SetVariables(Particles.GlowSize=&Multiply_Vector2D(A=$Particles.SpriteSize, B=<3 3>)) {isSetParameters=true}
   /SetVariables(Particles.GlowColor=&MultiplyLinearColor(
     Linear Color=$Particles.Color
     Scale Factor=&MakeVector4(X=1, Y=1, Z=1, W=&FN_FrontEndFloatModifier(Gameplay=0.095, Front End=0.035))
    )) {isSetParameters=true}
   /SetVariables(Particles.ShockColor:LinearColor=&MultiplyLinearColor(
     Linear Color=&MultiplyLinearColor(
      Linear Color=&ColorFromCurve(ColorCurve=DataInterface() {propertyValues {
         AlphaCurve {keys [<at 0 1 RCIM_Cubic>]}
         BlueCurve {keys [<at 0 20 RCIM_Cubic>,<at 0.6348 8>]}
         GreenCurve {keys [<at 0 20 RCIM_Cubic>,<at 0.6348 8>]}
         RedCurve {keys [<at 0 20 RCIM_Cubic>,<at 0.6348 8>]}
        }})
      Scale Factor=&MakeVector4_FromLinearColor(LinearColor=<rgba 0.2917 0.0611 0.0027 1>)
     )
     Scale Factor=&MakeFloat4FromVectorAndFloat(
      "Vector (RGB)"=<1 1 1>
      "Float (Alpha)"=&FloatFromCurve(
       FloatCurve=DataInterface() {propertyValues {Curve {keys [
           <at 0 0 RCIM_Cubic>
           {arriveTangent=-1.5276, interpMode=RCIM_Cubic, leaveTangent=-1.5276, tangentMode=RCTM_User, time=0.0466, value=1}
           {interpMode=RCIM_Cubic, tangentMode=RCTM_User, time=1, value=0}
          ]}}}
       Scale Curve=&FN_FrontEndFloatModifier(Gameplay=0.35, Front End=0.14)
      )
     )
    )) {isSetParameters=true}
   &ScaleMeshSize(Scale Factor=&VectorFromFloat(Value=&FloatFromCurve(
      FloatCurve=DataInterface() {propertyValues {Curve {keys [
          {arriveTangent=1.2547, interpMode=RCIM_Cubic, leaveTangent=1.2547, tangentMode=RCTM_User, time=0, value=0.3}
          {interpMode=RCIM_Cubic, tangentMode=RCTM_User, time=1, value=1.2}
         ]}}}
      Scale Curve=0.8
     )))
   /SetVariables(Particles.WaveDynMat:Vector4f=&MakeVector4(
     X=&FloatFromCurve(
      FloatCurve=DataInterface() {propertyValues {Curve {keys [{arriveTangent=1.5, interpMode=RCIM_Cubic, leaveTangent=1.5, tangentMode=RCTM_User, time=0, value=0.5659}, <at 1 1>]}}}
      Scale Curve=0.6
     )
     Y=&FloatFromCurve(FloatCurve=DataInterface() {propertyValues {Curve {keys [{arriveTangent=-0.0624, interpMode=RCIM_Cubic, leaveTangent=-0.0624, tangentMode=RCTM_User, time=0, value=1}, <at 1 0.996>]}}})
     Z=&FloatFromCurve(FloatCurve=DataInterface() {propertyValues {Curve {keys [<at 0 0 RCIM_Cubic>,<at 0.15 0.4572>,<at 1 0.5891>]}}}, Scale Curve=10)
     W=&FloatFromCurve(FloatCurve=DataInterface() {propertyValues {Curve {keys [<at 0 0.4427>,<at 1 1>]}}}, Scale Curve=0)
    )) {isSetParameters=true}
   &FN_Light_Attributes(
    Color=<rgba 0.7344 0.2277 0 1>
    Create Color Attribute=true
    Create Exponent Attribute=true
    Create Radius Attribute=true
    Exponent=1
    Radius=180
    Scale Color=&FloatFromCurve(FloatCurve=DataInterface() {propertyValues {Curve {keys [<at 0 1>,<at 1 0>]}}}, Scale Curve=20)
   )
   &CameraOffset(Camera Offset Amount=&Lerp_Float(
     A=30
     B=-5
     Alpha=&FloatFromCurve(FloatCurve=DataInterface() {propertyValues {Curve {keys [<at 0 0>,<at 0.25 1>]}}}, CurveIndex=$Engine.System.Age)
    ))
  ]
  rendererClasses [
   &NiagaraSpriteRendererProperties
   &NiagaraMeshRendererProperties
   &NiagaraLightRendererProperties
  ]
  renderers [
   {rendererClass=&NiagaraSpriteRendererProperties}
   {rendererClass=&NiagaraMeshRendererProperties, rendererIndex=1}
   {rendererClass=&NiagaraLightRendererProperties, rendererIndex=2}
  ]
 }
 LightiningBurst {
  emitterUpdateScript [
   &EmitterState(Inactive Response=::NewEnumerator1, Life Cycle Mode=::NewEnumerator1, Loop Behavior=::NewEnumerator1, Loop Duration=1, UseLoopDelay=false)
   &SpawnBurst_Instantaneous(Spawn Count=6, Spawn Time=0.33, Use Spawn Probability=false, UseLoopCountLimit=false)
  ]
  particleSpawnScript [
   &InitializeParticle(
    Color=&MaskLinearColorBySpawnGroup(
     Masked Linear Color=&RandomRangeLinearColor(Minimum=<rgba 0.6 0.6 0.6 0.5>, Maximum=<rgba 0.6 0.6 0.6 1>)
     Passthrough Linear Color=&RandomRangeLinearColor(Minimum=<rgba 0.8 0.8 0.8 0.5>, Maximum=<rgba 1 1 1 1>)
    )
    Lifetime=&RandomRangeFloat(Minimum=0.2, Maximum=0.3)
    Mesh Scale Mode=::NewEnumerator3
    Mesh Uniform Scale=&MaskFloatBySpawnGroup(
     Masked Float=&RandomRangeFloat(Minimum=3.5, Maximum=3.5)
     Passthrough Float=&RandomRangeFloat(Minimum=4, Maximum=4)
    )
    UsePositionOffset=false
    Write Material Random=true
   )
   &AddVelocity(Module.Use Falloff Distance=false, Override Seed=false, Velocity Mode=::NewEnumerator1)
   &SkeletalMeshLocation(Bone Sampling Mode=::NewEnumerator4, MaskSpawnGroup=false, Sampled Position Offset=<0 0 28>, Skeletal Mesh=$System.Skeletal Mesh)
   &InitialMeshOrientation(Mesh Axis To Orient=$Particles.Velocity, Mesh Orientation Mode=::NewEnumerator2, Orientation Vector=<0 0 1>, Use Rotation Vector=false)
  ]
  particleUpdateScript [
   &ParticleState()
   &ScaleMeshSize(Scale Factor=&RandomRangeVector(Minimum=<1 1 2>, Maximum=<2 1.5 2>)) {module=ScaleMeshSize001}
   &ScaleMeshSize(Scale Factor=&MakeVector(
     X=1
     Y=1
     Z=&FloatFromCurve(FloatCurve=DataInterface() {propertyValues {Curve {keys [
          {interpMode=RCIM_Cubic, leaveTangent=0.0966, tangentMode=RCTM_Break, time=0, value=1}
          {arriveTangent=0.0951, interpMode=RCIM_Cubic, leaveTangent=-0.1745, tangentMode=RCTM_Break, time=1, value=1.07}
         ]}}})
    ))
   &ScaleMeshSize(Scale Factor=&VectorFromFloat(Value=&RandomRangeFloat(Minimum=0.05, Maximum=0.08))) {module=ScaleMeshSize002}
   &DynamicMaterialParameters(
    Index 0 Param 4=&SelectFloatFromArray(
     Direct Array Index=&ModuloInt(A=&ReturnParticleIDAttributes(), B=4)
     Float Selection Array=DataInterface() {propertyValues {FloatData [3,0,10,3]}}
    )
    Index 1 Param 1=1.2
    Index 1 Param 2=2
    Index 1 Param 3=&FloatFromCurve(FloatCurve=DataInterface() {propertyValues {Curve {keys [{interpMode=RCIM_Cubic, tangentMode=RCTM_Break, time=0.5, value=-0.5}, <at 1 0.1>]}}})
    Index 2 Param 1=&Multiply_Float(
     A=&FloatFromCurve(
      FloatCurve=DataInterface() {propertyValues {Curve {keys [{interpMode=RCIM_Cubic, leaveTangent=8.1085, tangentMode=RCTM_Break, time=0.8568, value=1}, {tangentMode=RCTM_User, time=1, value=0}]}}}
      Scale Curve=0.35
     )
     B=&Divide_Float(
      A=&MakeFloatFromVector(Vector=$Particles.Scale, Channel=::NewEnumerator0)
      B=&RandomRangeFloat(Minimum=10, Maximum=15)
     )
    )
    Index 2 Param 2=&FloatFromCurve(
     FloatCurve=DataInterface() {propertyValues {Curve {keys [
         {interpMode=RCIM_Cubic, tangentMode=RCTM_Break, time=0.7, value=0}
         {arriveTangent=-0.0859, interpMode=RCIM_Cubic, tangentMode=RCTM_Break, time=1, value=-1}
        ]}}}
     Scale Curve=50
    )
    Index 2 Param 3=&Waveform(
     Global Input Position=&RandomRangeFloat(Minimum=0, Maximum=155)
     Global Phase=$Particles.Age
     "Waveform [1]"=::NewEnumerator12
     "[1] Phase Scale"=2
     "[1] Frequency"=1
     "[1] Amplitude Min/Max"=<0 1>
    )
    Index 2 Param 4=25
    Index 3 Param 1=&FloatFromCurve(
     FloatCurve=DataInterface() {propertyValues {Curve {keys [
         {interpMode=RCIM_Cubic, leaveTangent=-0.0025, tangentMode=RCTM_Break, time=0, value=0}
         {arriveTangent=1.8528, interpMode=RCIM_Cubic, tangentMode=RCTM_Break, time=1, value=1}
        ]}}}
     Scale Curve=&RandomRangeFloat(Minimum=15, Maximum=25)
    )
    Index 3 Param 2=&FloatFromCurve(FloatCurve=DataInterface() {propertyValues {Curve {keys [
         {interpMode=RCIM_Cubic, leaveTangent=-0.0025, tangentMode=RCTM_Break, time=0, value=0.5}
         {arriveTangent=-0.0316, interpMode=RCIM_Cubic, tangentMode=RCTM_Break, time=1, value=-0.2}
        ]}}})
    Index 3 Param 3=&FloatFromCurve(FloatCurve=DataInterface() {propertyValues {Curve {keys [
         {interpMode=RCIM_Cubic, leaveTangent=1.1034, tangentMode=RCTM_Break, time=0, value=-1}
         {arriveTangent=2.2399, tangentMode=RCTM_User, time=0.5, value=0.125}
        ]}}})
    Index 3 Param 4=&FloatFromCurve(FloatCurve=DataInterface() {propertyValues {Curve {keys [
         {interpMode=RCIM_Cubic, leaveTangent=3.0601, tangentMode=RCTM_Break, time=0.5, value=-0.5}
         {arriveTangent=3.2053, tangentMode=RCTM_User, time=1, value=1}
        ]}}})
    Module.Index0Param0Write=true
    Module.Index0Param1Write=true
    Module.Index0Param2Write=true
    Module.Index0Param3Write=true
    Module.Index1Param0Write=true
    Module.Index1Param1Write=true
    Module.Index1Param2Write=true
    Module.Index1Param3Write=true
    Module.Index2Param0Write=true
    Module.Index2Param1Write=true
    Module.Index2Param2Write=true
    Module.Index2Param3Write=true
    Module.Index3Param0Write=true
    Module.Index3Param1Write=true
    Module.Index3Param2Write=true
    Module.Index3Param3Write=true
    Write Parameter Index 1=true
    Write Parameter Index 2=true
    Write Parameter Index 3=true
   )
   &SolveForcesAndVelocity(Clamp Velocity=false, Limit Acceleration=false)
   &ScaleColor(
    Scale Alpha=&FloatFromCurve(FloatCurve=DataInterface() {propertyValues {Curve {keys [
         {tangentMode=RCTM_User, time=0, value=1}
         {arriveTangent=-0.0405, interpMode=RCIM_Cubic, leaveTangent=-0.0405, tangentMode=RCTM_User, time=0.7009, value=1}
         {arriveTangent=-8.0089, interpMode=RCIM_Cubic, leaveTangent=-8.0089, tangentMode=RCTM_User, time=1, value=0}
        ]}}})
    Scale RGB=&Lerp_Vector(
     A=&VectorFromFloat(Value=3)
     B=&VectorFromFloat(Value=&RandomRangeFloat(Minimum=1, Maximum=1.5, Evaluation Type=::NewEnumerator1))
     Alpha=&FloatFromCurve(FloatCurve=DataInterface() {propertyValues {Curve {keys [{interpMode=RCIM_Cubic, tangentMode=RCTM_Break, time=0, value=0}, {interpMode=RCIM_Cubic, tangentMode=RCTM_Break, time=0.333, value=1}]}}})
    )
    ScaleA=true
    ScaleRGB=true
   )
  ]
  rendererClasses [&NiagaraMeshRendererProperties]
  renderers [
   {rendererClass=&NiagaraMeshRendererProperties, rendererIndex=0}
   {rendererClass=&NiagaraMeshRendererProperties, rendererIndex=1}
  ]
 }
 LogoFlash {
  emitterUpdateScript [
   &EmitterState(Inactive Response=::NewEnumerator1, Life Cycle Mode=::NewEnumerator1, Loop Behavior=::NewEnumerator1, Loop Duration=1, UseLoopDelay=false)
   &SpawnBurst_Instantaneous(Spawn Count=6, Spawn Time=0.33, Use Spawn Probability=false, UseLoopCountLimit=false)
  ]
  particleSpawnScript [
   &FN_SetSpawnGroup(Number of Spawn Groups=6)
   &InitializeParticle(
    Color=&MergeHelper_ColorSwitch(
     0=<rgba 1.2 0.2163 0.0645 1>
     1=<rgba 1.2 0.2163 0.0645 1>
     2=<rgba 1.2 0.2163 0.0645 1>
     3=<rgba 4 3.396 2.7119 1>
     4=<rgba 10 3.7287 1.995 1>
     5=<rgba 0 0 0 1>
     Index=&SpawnGroup()
     Num=::NewEnumerator4
    )
    Lifetime=&MergeHelper_FloatSwitch(0=0.35, 1=0.62, 2=0.62, 3=0.4, 4=0.35, 5=0.4, Index=&SpawnGroup(), Num=::NewEnumerator4)
    Mesh Scale=<0.1 1 1.15>
    Mesh Scale Mode=::NewEnumerator4
    Sprite Rotation Angle=&MergeHelper_FloatSwitch(
     0=-90
     1=-90
     2=-90
     3=&RandomRangeFloat(Minimum=0, Maximum=1)
     4=&RandomRangeFloat(Minimum=0, Maximum=1)
     5=&RandomRangeFloat(Minimum=0, Maximum=1)
     Index=&SpawnGroup()
     Num=::NewEnumerator4
    )
    Sprite Rotation Mode=::NewEnumerator0
    Sprite Size Mode=::NewEnumerator3
    Uniform Sprite Size=&MergeHelper_FloatSwitch(0=55, 1=55, 2=55, 3=120, 4=180, 5=120, Index=&SpawnGroup(), Num=::NewEnumerator4)
    UsePositionOffset=false
    Write Material Random=true
   )
   &CameraOffset(Camera Offset Amount=&MergeHelper_FloatSwitch(0=15, 1=15, 2=15, 3=20, 4=50, 5=20, Index=&SpawnGroup(), Num=::NewEnumerator4))
   &InitialMeshOrientation(Mesh Orientation Mode=::NewEnumerator2, Rotation=<0 -0.002 0>, Use Rotation Vector=true)
   /SetVariables(Particles.Random=&RandomRangeFloat(Minimum=0, Maximum=1)) {isSetParameters=true, module=SetVariables_#2}
   /SetVariables(Particles.Random2:NiagaraFloat=&RandomRangeFloat(Minimum=0.5, Maximum=0.5)) {isSetParameters=true}
   /SetVariables(Particles.VisibilityTag=&SpawnGroup()) {isSetParameters=true}
   &SkeletalMeshLocation(
    Bone Sampling Mode=::NewEnumerator4
    MaskSpawnGroup=false
    Orientation Sampling=::NewEnumerator1
    Sampled Position Offset=&MergeHelper_VectorSwitch(0=<-6.5 -3.5 18>, 1=<-3 -1 29>, 2=<-3 -2 30>, 3=?, Index=&SpawnGroup(), Num=::NewEnumerator1)
    Skeletal Mesh=$System.Skeletal Mesh
   )
   /SetVariables() {isSetParameters=true}
  ]
  particleUpdateScript [
   &ParticleState()
   &ScaleSpriteSize(
    Scale Sprite Size Mode=::NewEnumerator0
    Uniform Scale Factor=&MergeHelper_FloatSwitch(
     0=&FloatFromCurve(FloatCurve=DataInterface() {propertyValues {Curve {keys [
          {arriveTangent=0.603, interpMode=RCIM_Cubic, leaveTangent=0.603, tangentMode=RCTM_User, time=0, value=0.7}
          {tangentMode=RCTM_User, time=1, value=1}
         ]}}})
     1=&FloatFromCurve(FloatCurve=DataInterface() {propertyValues {Curve {keys [
          {arriveTangent=0.603, interpMode=RCIM_Cubic, leaveTangent=0.603, tangentMode=RCTM_User, time=0, value=0.7}
          {tangentMode=RCTM_User, time=1, value=1}
         ]}}})
     2=&FloatFromCurve(FloatCurve=DataInterface() {propertyValues {Curve {keys [
          {arriveTangent=0.603, interpMode=RCIM_Cubic, leaveTangent=0.603, tangentMode=RCTM_User, time=0, value=0.7}
          {tangentMode=RCTM_User, time=1, value=1}
         ]}}})
     3=&FloatFromCurve(
      FloatCurve=DataInterface() {propertyValues {Curve {keys [
          {arriveTangent=2.6083, interpMode=RCIM_Cubic, leaveTangent=0.4318, tangentMode=RCTM_User, time=0, value=0.5}
          {arriveTangent=1.3014, interpMode=RCIM_Cubic, leaveTangent=1.3014, tangentMode=RCTM_User, time=0.3, value=0.8}
          {arriveTangent=0.1331, leaveTangent=0.1331, time=1, value=1.2}
         ]}}}
      Scale Curve=1.25
     )
     4=&FloatFromCurve(
      FloatCurve=DataInterface() {propertyValues {Curve {keys [
          {arriveTangent=2.6083, interpMode=RCIM_Cubic, leaveTangent=0.4318, tangentMode=RCTM_User, time=0, value=0.5}
          {arriveTangent=1.3014, interpMode=RCIM_Cubic, leaveTangent=1.3014, tangentMode=RCTM_User, time=0.3, value=0.8}
          {arriveTangent=0.1331, leaveTangent=0.1331, time=1, value=1.2}
         ]}}}
      Scale Curve=1.25
     )
     5=&FloatFromCurve(
      FloatCurve=DataInterface() {propertyValues {Curve {keys [
          {arriveTangent=2.6083, interpMode=RCIM_Cubic, leaveTangent=0.4318, tangentMode=RCTM_User, time=0, value=0.5}
          {arriveTangent=1.3014, interpMode=RCIM_Cubic, leaveTangent=1.3014, tangentMode=RCTM_User, time=0.3, value=0.8}
          {arriveTangent=0.1331, leaveTangent=0.1331, time=1, value=1.2}
         ]}}}
      Scale Curve=1.25
     )
     Index=&SpawnGroup()
     Num=::NewEnumerator4
    )
   )
   &ScaleMeshSize(Scale Factor=&Multiply_VectorByFloat(Vector=<1 1 1>, Float=0.7))
   &ScaleColor(
    Scale Alpha=&MergeHelper_FloatSwitch(
     0=&Multiply_Float(
      A=10
      B=&FloatFromCurve(FloatCurve=DataInterface() {propertyValues {Curve {keys [<at 0 0>,<at 0.1 0>,<at 0.2 0.8>,<at 1 0>]}}})
     )
     1=0
     2=0
     3=&FloatFromCurve(
      FloatCurve=DataInterface() {propertyValues {Curve {keys [{arriveTangent=-3.3102, interpMode=RCIM_Cubic, time=0, value=0}, <at 0.5 1 RCIM_Cubic>, <at 1 0 RCIM_Cubic>]}}}
      Scale Curve=&Multiply_Float(
       A=&FN_FrontEndFloatModifier(Gameplay=1, Front End=0.5)
       B=&FN_PickFloatDependOnPlatform(TargetPlatform=DataInterface() {propertyValues {Platforms {qualityLevelMask=-29}}}, TargetPlatformActive=0.65, TargetPlatformNoActive=1)
      )
     )
     4=&Multiply_Float(
      A=&FN_FrontEndFloatModifier(Gameplay=0.2, Front End=0.1)
      B=&FN_PickFloatDependOnPlatform(TargetPlatform=DataInterface() {propertyValues {Platforms {qualityLevelMask=-29}}}, TargetPlatformActive=0.65, TargetPlatformNoActive=1)
     )
     5=&FloatFromCurve(
      FloatCurve=DataInterface() {propertyValues {Curve {keys [{interpMode=RCIM_Cubic, tangentMode=RCTM_User, time=0, value=1}, {interpMode=RCIM_Cubic, tangentMode=RCTM_User, time=1, value=0}]}}}
      Scale Curve=&FN_FrontEndFloatModifier(Gameplay=1, Front End=0.6)
     )
     Index=&SpawnGroup()
     Num=::NewEnumerator4
    )
    Scale RGB=&MergeHelper_VectorSwitch(
     0=&Multiply_VectorByFloat(
      Vector=&Multiply_VectorByFloat(Vector=<1 1 1>, Float=1)
      Float=&Multiply_Float(
       A=&RandomRangeFloat(Minimum=0, Maximum=1, Evaluation Type=::NewEnumerator1)
       B=&Multiply_Float(
        A=1
        B=&FloatFromCurve(
         FloatCurve=DataInterface() {propertyValues {Curve {keys [<at 0 10>,<at 1 10>]}}}
         Scale Curve=&FNTimeOfDayScalar(Night Scale=0.3, Morning Scale=0.7, Evening Scale=0.7)
        )
       )
      )
     )
     1=<0 0 0>
     2=<0 0 0>
     3=&MakeVectorFromLinearColorRGB(Color=&ColorFromCurve(ColorCurve=DataInterface() {propertyValues {
         AlphaCurve {keys []}
         BlueCurve {keys [<at 0 1.8256 RCIM_Cubic>,<at 0.3877 0>,<at 1 0.0229 RCIM_Cubic>]}
         GreenCurve {keys [<at 0 2.8793 RCIM_Cubic>,<at 0.3877 0.225>,<at 1 0.0716 RCIM_Cubic>]}
         RedCurve {keys [<at 0 10 RCIM_Cubic>,<at 0.3877 1>,<at 1 0.1878 RCIM_Cubic>]}
        }}))
     4=&VectorFromFloat(Value=&FloatFromCurve(FloatCurve=DataInterface() {propertyValues {Curve {keys [{interpMode=RCIM_Cubic, tangentMode=RCTM_User, time=0, value=1}, {interpMode=RCIM_Cubic, tangentMode=RCTM_User, time=1, value=0}]}}}))
     5=<1 1 1>
     Index=&SpawnGroup()
     Num=::NewEnumerator4
    )
    ScaleA=true
    ScaleRGB=true
   )
   &DynamicMaterialParameters(
    Index 0 Param 1=&FloatFromCurve(FloatCurve=DataInterface() {propertyValues {Curve {keys [
         {arriveTangent=4.2221, interpMode=RCIM_Cubic, leaveTangent=4.2221, tangentMode=RCTM_User, time=0, value=0}
         {tangentMode=RCTM_User, time=0.4298, value=1}
        ]}}})
    Index 0 Param 2=0.7
    Index 0 Param 3=&FloatFromCurve(FloatCurve=DataInterface() {propertyValues {Curve {keys [
         {arriveTangent=4.0072, interpMode=RCIM_Cubic, leaveTangent=4.0072, tangentMode=RCTM_User, time=0, value=0}
         {arriveTangent=4.2299, interpMode=RCIM_Cubic, leaveTangent=4.2299, time=0.2543, value=0.6225}
         {arriveTangent=3.1884, interpMode=RCIM_Cubic, leaveTangent=3.1884, time=0.4395, value=1.8588}
         {arriveTangent=1.6993, interpMode=RCIM_Cubic, leaveTangent=1.6993, tangentMode=RCTM_User, time=1, value=3}
        ]}}})
    Index 0 Param 4=5
    Index 1 Param 1=&Multiply_Float(A=&Lerp_Float(A=4, B=4, Alpha=$Particles.Random), B=1)
    Index 1 Param 3=1.3
    Index 1 Param 4=1
    Index 2 Param 1=&FloatFromCurve(FloatCurve=DataInterface() {propertyValues {Curve {keys [<at 0 1>,<at 1 0.3>]}}})
    Module.Param0WriteEnabled=true
    Module.Param0WriteEnabled1=true
    Module.Param0WriteEnabled2=true
    Module.Param1WriteEnabled=true
    Module.Param1WriteEnabled1=true
    Module.Param1WriteEnabled2=true
    Module.Param2WriteEnabled=true
    Module.Param2WriteEnabled1=true
    Module.Param2WriteEnabled2=true
    Module.Param3WriteEnabled=true
    Module.Param3WriteEnabled1=true
    Module.Param3WriteEnabled2=true
    Write Parameter Index 1=true
    Write Parameter Index 2=true
   )
   /SetVariables(Particles.Lightning glow color:LinearColor=&MakeLinearColorFromVectorAndFloat(
     "Vector (RGB)"=&MakeVectorFromLinearColorRGB(Color=&MultiplyLinearColor(
       Linear Color=<rgba 1 0.1364 0.0661 1>
       Scale Factor=&MakeVector4_FromFloat(Float=&RandomRangeFloat(Minimum=0.5, Maximum=1, Evaluation Type=::NewEnumerator1))
      ))
     "Float (Alpha)"=&Multiply_Float(
      A=&FloatFromCurve(
       FloatCurve=DataInterface() {propertyValues {Curve {keys [<at 0 1 RCIM_Cubic>, {interpMode=RCIM_Cubic, leaveTangent=-3.6307, time=0.9, value=0}]}}}
       Scale Curve=&FNTimeOfDayScalar(Night Scale=0.5, Morning Scale=0.7, Evening Scale=0.7)
      )
      B=&FN_PickFloatDependOnPlatform(TargetPlatformActive=10, TargetPlatformNoActive=3)
     )
    )) {isSetParameters=true, module=SetVariables_#3}
   /SetVariables(Particles.Glow Dy02:Vector4f=&MakeVector4(
     X=&FloatFromCurve(FloatCurve=DataInterface() {propertyValues {Curve {keys [{tangentMode=RCTM_User, time=0, value=0.2}, {tangentMode=RCTM_User, time=0.5, value=1}]}}})
     Y=0.7
     Z=0
     W=5
    )) {isSetParameters=true, module=SetVariables_#4}
   /SetVariables(Particles.Glow Dy03:Vector4f=&MakeVector4(
     X=&FloatFromCurve(
      FloatCurve=DataInterface() {propertyValues {Curve {keys [<at 0 1>,<at 1 0>]}}}
      Scale Curve=&Multiply_Float(A=&Lerp_Float(A=6, B=6, Alpha=$Particles.Random), B=1)
     )
     Y=0
     Z=1.3
     W=1
    )) {isSetParameters=true, module=SetVariables_#5}
   /SetVariables(
    Particles.LogoColor:LinearColor=&MakeLinearColorFromVectorAndFloat(
     "Vector (RGB)"=&MakeVectorFromLinearColorRGB(Color=<rgba 6 1.4246 0.7188 1>)
     "Float (Alpha)"=&FloatFromCurve(
      FloatCurve=DataInterface() {propertyValues {Curve {keys [
          {arriveTangent=0.2357, interpMode=RCIM_Cubic, time=0, value=0}
          {arriveTangent=0.3592, interpMode=RCIM_Cubic, leaveTangent=0.3592, time=0.2, value=0.5883}
          {arriveTangent=-0.8922, interpMode=RCIM_Cubic, leaveTangent=-0.8922, tangentMode=RCTM_User, time=0.4262, value=0.1531}
          <at 1 0>
         ]}}}
      Scale Curve=3
     )
    )
    Particles.LogoDarkColor:LinearColor=&MakeLinearColorFromVectorAndFloat(
     "Vector (RGB)"=&MakeVectorFromLinearColorRGB(Color=<rgba 0.1354 0.0322 0.0162 1>)
     "Float (Alpha)"=&FloatFromCurve(
      FloatCurve=DataInterface() {propertyValues {Curve {keys [
          {arriveTangent=0.2357, interpMode=RCIM_Cubic, time=0, value=0}
          {arriveTangent=0.3592, interpMode=RCIM_Cubic, leaveTangent=0.3592, time=0.2, value=0.5883}
          {arriveTangent=-0.8922, interpMode=RCIM_Cubic, leaveTangent=-0.8922, tangentMode=RCTM_User, time=0.4262, value=0.1531}
          <at 1 0>
         ]}}}
      Scale Curve=5
     )
    )
   ) {isSetParameters=true}
   &SpriteFacingAndAlignment(Sprite Alignment=<1 0 0>, Sprite Facing=<0 1 0>, WriteAlignment=true, WriteFacing=true)
  ]
  rendererClasses [&NiagaraMeshRendererProperties, &NiagaraSpriteRendererProperties]
  renderers [
   {rendererClass=&NiagaraMeshRendererProperties, rendererIndex=0}
   {rendererClass=&NiagaraMeshRendererProperties, rendererIndex=1}
   {rendererClass=&NiagaraSpriteRendererProperties, rendererIndex=2}
   {rendererClass=&NiagaraSpriteRendererProperties, rendererIndex=3}
   {rendererClass=&NiagaraSpriteRendererProperties, rendererIndex=4}
   {rendererClass=&NiagaraSpriteRendererProperties, rendererIndex=5}
   {rendererClass=&NiagaraSpriteRendererProperties, rendererIndex=6}
  ]
 }
 Sparks {
  emitterUpdateScript [
   &EmitterState(Inactive Response=::NewEnumerator1, Life Cycle Mode=::NewEnumerator1, Loop Behavior=::NewEnumerator1, Loop Duration=2, UseLoopDelay=false)
   &SpawnBurst_Instantaneous(Spawn Count=15, Spawn Time=1.22, Use Spawn Probability=false, UseLoopCountLimit=false)
  ]
  particleSpawnScript [
   &InitializeParticle(
    Color=&ColorFromCurve(ColorCurve=DataInterface() {propertyValues {
       AlphaCurve {keys [<at 0 1 RCIM_Cubic>]}
       BlueCurve {keys [<at 0.4264 1 RCIM_Cubic>,<at 0.4705 1 RCIM_Cubic>]}
       GreenCurve {keys [<at 0.4264 0.6043 RCIM_Cubic>,<at 0.4705 0.2355 RCIM_Cubic>]}
       RedCurve {keys [<at 0.4264 0 RCIM_Cubic>,<at 0.4705 0 RCIM_Cubic>]}
      }})
    Lifetime Max=0.7
    Lifetime Min=0.5
    Lifetime Mode=::NewEnumerator1
    Sprite Rotation Mode=::NewEnumerator2
    Sprite Size=&RandomRangeVector2D(Minimum=<0.8 4>, Maximum=<1.5 8>)
    Sprite Size Mode=::NewEnumerator4
    UsePositionOffset=false
    Write Material Random=true
   )
   &SkeletalMeshLocation(
    Bone / Socket Index=2
    Bone Sampling Mode=::NewEnumerator4
    MaskSpawnGroup=false
    Orientation Sampling=::NewEnumerator1
    Sampled Position Offset=&RandomRangeVector(Minimum=<0 20 0>, Maximum=<0 100 0>)
    Skeletal Mesh=$System.Skeletal Mesh
   )
   &AddVelocity(
    Cone Axis=&Lerp_Vector(
     A=<-1 0 0>
     B=<1 0 0>
     Alpha=&UniformAOrBFloat(A=0, Evaluation Type=::NewEnumerator0, Recalculate Random Each Loop=true, Override Seed=false)
    )
    Module.UseConeFalloff=false
    Module.UseVelDistribution=false
    Override Seed=false
    Velocity Mode=::NewEnumerator2
    Velocity Speed=&RandomRangeFloat(Minimum=600, Maximum=950)
   )
  ]
  particleUpdateScript [
   /SetVariables(Particles.GlitterAlpha:NiagaraFloat=&FloatFromCurve(
     FloatCurve=DataInterface() {propertyValues {Curve {keys [
         {interpMode=RCIM_Cubic, tangentMode=RCTM_User, time=0, value=0}
         {interpMode=RCIM_Cubic, tangentMode=RCTM_User, time=0.02, value=1}
         {interpMode=RCIM_Cubic, tangentMode=RCTM_User, time=0.25, value=0}
        ]}}}
     CurveIndex=&ModuloFloat(
      A=&Add_Float(
       A=&Multiply_Float(A=$Emitter.Age, B=2)
       B=&RandomRangeFloat(Minimum=0, Maximum=1)
      )
      B=1
     )
    )) {isSetParameters=true}
   &ParticleState()
   &ScaleSpriteSize(
    Uniform Curve Scale=0.7
    Uniform Curve Sprite Scale=DataInterface() {propertyValues {Curve {keys [
        {interpMode=RCIM_Cubic, tangentMode=RCTM_User, time=0, value=0}
        {interpMode=RCIM_Cubic, tangentMode=RCTM_User, time=0.1, value=1}
        {arriveTangent=-0.2407, interpMode=RCIM_Cubic, leaveTangent=-0.2407, tangentMode=RCTM_User, time=1, value=0}
       ]}}}
   )
   &SpriteRotationRate(Rotation Rate=&RandomRangeFloat(Minimum=-20, Maximum=20))
   &Color(Color=&MultiplyLinearColor(
     Linear Color=&MultiplyLinearColor(
      Linear Color=&ColorFromCurve(ColorCurve=DataInterface() {propertyValues {
         AlphaCurve {keys [<at 0.4106 1 RCIM_Cubic>,<at 1 0 RCIM_Cubic>]}
         BlueCurve {keys [<at 0 10 RCIM_Cubic>,<at 0.1841 1>,<at 0.447 0.3802 RCIM_Cubic>]}
         GreenCurve {keys [<at 0 10 RCIM_Cubic>,<at 0.1841 1>,<at 0.447 0.3802 RCIM_Cubic>]}
         RedCurve {keys [<at 0 10 RCIM_Cubic>,<at 0.1841 1>,<at 0.447 0.3802 RCIM_Cubic>]}
        }})
      Scale Factor=&MakeVector4_FromLinearColor(LinearColor=&LerpLinearColors(StartColor=<rgba 1 1 1 1>, EndColor=<rgba 5 5 5 1>, LerpFactor=$Particles.GlitterAlpha))
     )
     Scale Factor=&MakeVector4_FromLinearColor(LinearColor=<rgba 1 0.2674 0 1>)
    ))
   &GravityForce(Gravity=<0 0 -300>)
   /SetVariables(Particles.GlowColor=&MultiplyLinearColor(Linear Color=$Particles.Color, Scale Factor=<1 1 1 0.08>)) {isSetParameters=true}
   /SetVariables(Particles.GlowSize=&Multiply_Vector2D(A=$Particles.SpriteSize, B=<2 2>)) {isSetParameters=true}
   &Drag(
    Drag=&Lerp_Float(
     A=&RandomRangeFloat(Minimum=3, Maximum=4)
     B=&RandomRangeFloat(Minimum=6, Maximum=8)
     Alpha=$Particles.SizeAlpha
    )
    Use Linear Drag=true
    Use Rotational Drag=true
   )
   &CurlNoiseForce(
    Noise Strength=&FloatFromCurve(
     FloatCurve=DataInterface() {propertyValues {Curve {keys [
         {arriveTangent=2.347, interpMode=RCIM_Cubic, leaveTangent=2.347, tangentMode=RCTM_User, time=0, value=0}
         {interpMode=RCIM_Cubic, tangentMode=RCTM_User, time=1, value=1}
        ]}}}
     Scale Curve=1500
    )
    Pan Noise=false
   )
   &SolveForcesAndVelocity(Clamp Velocity=false, Limit Acceleration=false)
   &DynamicMaterialParameters(
    Index 0 Param 1=1
    Index 0 Param 2=0.5
    Index 0 Param 3=5
    Module.Index0Param0Write=true
    Module.Index0Param1Write=true
    Module.Index0Param2Write=true
    Module.Index0Param3Write=true
   )
  ]
  rendererClasses [&NiagaraSpriteRendererProperties]
  renderers [
   {rendererClass=&NiagaraSpriteRendererProperties}
   {rendererClass=&NiagaraSpriteRendererProperties, rendererIndex=1}
  ]
 }
}
summary {
 systemName=NS_Emote_SwordSplit_01
 emitters [
  {emitterName=LogoFlash, rendererClasses [&NiagaraMeshRendererProperties, &NiagaraSpriteRendererProperties]}
  {emitterName=EnergyWave, rendererClasses [&NiagaraMeshRendererProperties, &NiagaraLightRendererProperties]}
  {emitterName=LightiningBurst, rendererClasses [&NiagaraMeshRendererProperties]}
  {emitterName=ConnectingStreaks, rendererClasses [&NiagaraMeshRendererProperties]}
  {emitterName=CapsuleGlow, rendererClasses [&NiagaraSpriteRendererProperties]}
  {emitterName=Sparks, rendererClasses [&NiagaraSpriteRendererProperties]}
  {
   emitterName=Flash
   rendererClasses [
    &NiagaraSpriteRendererProperties
    &NiagaraMeshRendererProperties
    &NiagaraLightRendererProperties
   ]
  }
 ]
 userVariables [
  User.bIsFrontEnd {type=NiagaraBool}
  User.bIsFrontEndPreview {type=NiagaraBool}
  User.EmoteSkeleton {type=NiagaraFloat, defaultValue=0, description=""}
 ]
}
