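# Niagara system for the SwordSplit emote, drawing orange electrical energy effects on bones or sockets of a skeletal mesh.
# ConnectingArcs and EdgeGlow run along two sockets, with arcs shrinking away by 0.6 seconds as LightiningBurst and Flare fire.
# EnergyWave spawns mesh waves with lights at start and at 0.64 seconds.
# Spawn-group switches vary socket, size and color per particle; front-end and platform modifiers adjust intensity.
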NiagaraSystem /BRCosmetics/Animation/Game/MainPlayer/Emotes/SwordSplit/FX/NS_Emote_SwordSplit_02
dependencies {
 usedDataInterfaces [
  &NiagaraDataInterfaceCurve
  &NiagaraDataInterfacePlatformSet
  &NiagaraDataInterfaceColorCurve
  &NiagaraDataInterfaceArrayFloat
  &NiagaraDataInterfaceArrayColor
  &NiagaraDataInterfaceSkeletalMesh
 ]
 usedDynamicInputs [&MakeVector4_FromLinearColor]
 usedModules [
  &self:SystemSpawnScript.SystemScriptSource.SystemScriptGraph.NiagaraNodeAssignment_0.SetVariables_#1
  &SystemState
 ]
}
emitters {
 ConnectingArcs {
  emitterUpdateScript [
   &EmitterState(Inactive Response=::NewEnumerator1, Life Cycle Mode=::NewEnumerator1, Loop Behavior=::NewEnumerator1, Loop Duration=3, UseLoopDelay=false)
   &SpawnRate(
    SpawnRate=&Multiply_Float(
     A=12
     B=&FloatFromCurve(
      FloatCurve=DataInterface() {propertyValues {Curve {keys [
          <at 0 0 RCIM_Cubic>
          {arriveTangent=1.8541, interpMode=RCIM_Cubic, leaveTangent=1.8541, tangentMode=RCTM_User, time=0.1, value=0.9}
          {arriveTangent=-0.7169, interpMode=RCIM_Cubic, leaveTangent=-0.7169, time=0.8713, value=1}
          <at 1.3555 0>
         ]}}}
      CurveIndex=$System.Age
     )
    )
    Use Spawn Probability=false
   )
   &SpawnRate(
    SpawnGroup=1
    SpawnRate=&Multiply_Float(
     A=12
     B=&FloatFromCurve(
      FloatCurve=DataInterface() {propertyValues {Curve {keys [
          <at 0 0 RCIM_Cubic>
          {arriveTangent=1.8541, interpMode=RCIM_Cubic, leaveTangent=1.8541, tangentMode=RCTM_User, time=0.1, value=0.9}
          {arriveTangent=-0.7169, interpMode=RCIM_Cubic, leaveTangent=-0.7169, time=0.8713, value=1}
          <at 1.3555 0>
         ]}}}
      CurveIndex=$System.Age
     )
    )
    Use Spawn Probability=false
   ) {module=SpawnRate002}
  ]
  particleSpawnScript [
   /SetVariables(Particles.SizeToPosLerp:NiagaraFloat=&RandomRangeFloat(Minimum=0, Maximum=1)) {isSetParameters=true}
   /SetVariables(Particles.RandFloat:NiagaraFloat=&UniformAOrBFloat(
     A=&RandomRangeFloat(Minimum=0, Maximum=0.45)
     B=&RandomRangeFloat(Minimum=0.45, Maximum=1)
     Evaluation Type=::NewEnumerator0
     Recalculate Random Each Loop=true
     Override Seed=false
    )) {isSetParameters=true}
   &InitializeParticle(
    Color=<rgba 1 0.1802 0.0537 1>
    Lifetime Max=0.4
    Lifetime Min=0.2
    Lifetime Mode=::NewEnumerator1
    Mesh Scale=&Multiply_Vector(
     A=&RandomRangeVector(Minimum=<1 0.1 0.3>, Maximum=<1 0.1 0.5>)
     B=&VectorFromFloat(Value=&Lerp_Float(A=0.8, B=1.4, Alpha=$Particles.SizeToPosLerp))
    )
    Mesh Scale Mode=::NewEnumerator4
    UsePositionOffset=false
    Write Material Random=true
   )
   &CameraOffset(Camera Offset Amount=14)
   /SetVariables(Particles.Random=&RandomRangeFloat(Minimum=0, Maximum=1)) {isSetParameters=true}
   &InitialMeshOrientation(
    Mesh Axis To Orient=&MergeHelper_VectorSwitch(0=<0 1 0>, 1=<0 1 0>, 2=<1 0 0>, 3=?, Index=&SpawnGroup(), Num=::NewEnumerator1)
    Mesh Orientation Mode=::NewEnumerator2
    Orientation Vector=&MergeHelper_VectorSwitch(0=<-1 0 0>, 1=<-1 0 0>, 2=<1 0 0>, 3=?, Index=&SpawnGroup(), Num=::NewEnumerator1)
    Rotation=&MergeHelper_VectorSwitch(
     0=<0 -0.0695 0>
     1=<0 -0.06 -0.0869>
     2=&RandomRangeVector(Minimum=<0 0 0>, Maximum=<1 1 1>)
     3=?
     Index=&SpawnGroup()
     Num=::NewEnumerator1
    )
    Use Rotation Vector=true
   )
  ]
  particleUpdateScript [
   &ParticleState()
   &ScaleMeshSize(Scale Factor=&VectorFromFloat(Value=&FloatFromCurve(
      FloatCurve=DataInterface() {propertyValues {Curve {keys [<at 0.4438 1 RCIM_Constant>,<at 0.6 0>]}}}
      CurveIndex=$System.Age
      Scale Curve=&MaskFloatBySpawnGroup(Masked Float=1, Passthrough Float=1)
     ))) {module=ScaleMeshSize001}
   &ScaleMeshSize(Scale Factor=&Multiply_VectorByFloat(
     Vector=<1 1 1>
     Float=&FloatFromCurve(FloatCurve=DataInterface() {propertyValues {Curve {keys [<at 0 1.75>,<at 1 1>]}}}, CurveIndex=$System.Age)
    ))
   &ScaleColor(
    Scale Alpha=&Multiply_Float(
     A=10
     B=&FloatFromCurve(
      FloatCurve=DataInterface() {propertyValues {Curve {keys [<at 0 0>,<at 0.1 0>,<at 0.2 0.8>,<at 1 0>]}}}
      Scale Curve=&FN_FrontEndFloatModifier(Gameplay=1, Front End=0.7)
     )
    )
    Scale RGB=&Multiply_VectorByFloat(
     Vector=<1 1 1>
     Float=&Multiply_Float(
      A=&RandomRangeFloat(Minimum=0, Maximum=1, Evaluation Type=::NewEnumerator1)
      B=&Multiply_Float(
       A=1
       B=&FloatFromCurve(FloatCurve=DataInterface() {propertyValues {Curve {keys [<at 0 10>,<at 1 10>]}}})
      )
     )
    )
    ScaleA=true
    ScaleRGB=true
   )
   &SkeletalMeshLocation(
    Bone / Socket Index=&MergeHelper_IntSwitch(
     0=0
     1=1
     2=&RandomRangeInt(Minimum=0, Maximum=1)
     3=?
     4=?
     5=?
     Index=&SpawnGroup()
     Num=::NewEnumerator1
    )
    Bone Sampling Mode=::NewEnumerator4
    MaskSpawnGroup=false
    Sampled Position Offset=&Lerp_Vector(A=<0 0 20>, B=<1 1 87>, Alpha=$Particles.SizeToPosLerp)
    Skeletal Mesh=$System.Skeletal Mesh
   )
   &DynamicMaterialParameters(
    Index 0 Param 1=&FloatFromCurve(FloatCurve=DataInterface() {propertyValues {Curve {keys [{tangentMode=RCTM_User, time=0, value=0.2}, {tangentMode=RCTM_User, time=1, value=1}]}}})
    Index 0 Param 2=0.7
    Index 0 Param 3=&FloatFromCurve(FloatCurve=DataInterface() {propertyValues {Curve {keys [
         {arriveTangent=12.4916, interpMode=RCIM_Cubic, leaveTangent=12.4916, tangentMode=RCTM_User, time=0, value=0}
         {arriveTangent=3, interpMode=RCIM_Cubic, leaveTangent=3, time=0.0944, value=2.0838}
         {arriveTangent=1.6993, interpMode=RCIM_Cubic, leaveTangent=1.6993, tangentMode=RCTM_User, time=1, value=3}
        ]}}})
    Index 0 Param 4=3
    Index 1 Param 1=&Multiply_Float(A=2.33, B=1)
    Index 1 Param 2=$Particles.Random
    Index 1 Param 3=1
    Index 1 Param 4=1
    Index 2 Param 1=&FloatFromCurve(FloatCurve=DataInterface() {propertyValues {Curve {keys [<at 0 1>,<at 1 0.3>]}}})
    Module.Param0WriteEnabled=true
    Module.Param0WriteEnabled1=true
    Module.Param0WriteEnabled2=true
    Module.Param1WriteEnabled=true
    Module.Param1WriteEnabled1=true
    Module.Param1WriteEnabled2=true
    Module.Param2WriteEnabled=true
    Module.Param2WriteEnabled1=true
    Module.Param2WriteEnabled2=true
    Module.Param3WriteEnabled=true
    Module.Param3WriteEnabled1=true
    Module.Param3WriteEnabled2=true
    Write Parameter Index 1=true
    Write Parameter Index 2=true
   )
   /SetVariables(Particles.Lightning glow color:LinearColor=&MakeLinearColorFromVectorAndFloat(
     "Vector (RGB)"=&MakeVectorFromLinearColorRGB(Color=&MultiplyLinearColor(
       Linear Color=<rgba 1 0.1558 0.0312 1>
       Scale Factor=&MakeVector4_FromFloat(Float=&FN_FrontEndFloatModifier(Gameplay=0.5, Front End=0.4))
      ))
     "Float (Alpha)"=&Multiply_Float(
      A=&FloatFromCurve(FloatCurve=DataInterface() {propertyValues {Curve {keys [<at 0 1 RCIM_Cubic>, {interpMode=RCIM_Cubic, leaveTangent=-3.6307, time=0.9, value=0}]}}})
      B=2
     )
    )) {isSetParameters=true}
   /SetVariables(Particles.Glow Dy02:Vector4f=&MakeVector4(
     X=&FloatFromCurve(FloatCurve=DataInterface() {propertyValues {Curve {keys [{tangentMode=RCTM_User, time=0, value=0.2}, {tangentMode=RCTM_User, time=0.5, value=1}]}}})
     Y=0.7
     Z=0
     W=3
    )) {isSetParameters=true}
   /SetVariables(Particles.Glow Dy03:Vector4f=&MakeVector4(X=&Multiply_Float(A=1.3, B=1), Y=$Particles.Random, Z=1, W=1)) {isSetParameters=true}
   &FN_Sprite_Attributes(
    Camera Offset=20
    Color=&MakeLinearColorFromVectorAndFloat(
     "Vector (RGB)"=&Multiply_Vector(A=&MakeVectorFromLinearColorRGB(Color=<rgba 1 0.141 0.0295 1>), B=<1 1 1>)
     "Float (Alpha)"=&Multiply_Float(
      A=&FloatFromCurve(FloatCurve=DataInterface() {propertyValues {Curve {keys [
           {interpMode=RCIM_Cubic, tangentMode=RCTM_User, time=0, value=0}
           {interpMode=RCIM_Cubic, tangentMode=RCTM_User, time=0.3, value=1}
           {tangentMode=RCTM_User, time=1, value=0}
          ]}}})
      B=1
     )
    )
    Create Camera Offset Attribute=true
    Create Color Attribute=true
    Create SpriteSize Attribute=true
    SpriteSize=&Multiply_Vector2D_ByFloat(
     Float=&FloatFromCurve(
      FloatCurve=DataInterface() {propertyValues {Curve {keys [
          {interpMode=RCIM_Cubic, tangentMode=RCTM_User, time=0, value=0}
          {interpMode=RCIM_Cubic, tangentMode=RCTM_User, time=0.3, value=1}
          {tangentMode=RCTM_User, time=1, value=0}
         ]}}}
      Scale Curve=&FloatFromCurve(FloatCurve=DataInterface() {propertyValues {Curve {keys [<at 0.4438 1 RCIM_Constant>,<at 0.6 0>]}}}, CurveIndex=$System.Age)
     )
     Vector2D=<40 40>
    )
    Write to Module Namespace=true
   )
   /SetVariables(Particles.DecalDynamic2:Vector4f=<0 0 1 1>) {isSetParameters=true}
  ]
  rendererClasses [&NiagaraMeshRendererProperties, &NiagaraSpriteRendererProperties]
  renderers [
   {rendererClass=&NiagaraMeshRendererProperties, rendererIndex=0}
   {rendererClass=&NiagaraMeshRendererProperties, rendererIndex=1}
   {rendererClass=&NiagaraSpriteRendererProperties, rendererIndex=2}
  ]
 }
 EdgeGlow {
  emitterUpdateScript [
   &EmitterState(Inactive Response=::NewEnumerator1, Life Cycle Mode=::NewEnumerator1, Loop Behavior=::NewEnumerator1, Loop Duration=3, UseLoopDelay=false)
   &SpawnBurst_Instantaneous(Spawn Count=4, Use Spawn Probability=false, UseLoopCountLimit=false)
  ]
  particleSpawnScript [
   &FN_SetSpawnGroup()
   &InitializeParticle(
    Color=&MergeHelper_ColorSwitch(
     0=<rgba 1 0.2384 0.0009 0.2>
     1=<rgba 1 0.2384 0.0009 0.2>
     2=<rgba 5 1.1918 0.0043 1>
     3=<rgba 5 1.1918 0.0043 1>
     Index=&SpawnGroup()
     Num=::NewEnumerator2
    )
    Lifetime=0.65
    Position Mode=::NewEnumerator0
    Sprite Rotation Mode=::NewEnumerator0
    Sprite Size=&MergeHelper_Vector2Switch(0=<22.1109 86>, 1=<22.1109 98>, 2=<5 100>, 3=<5 115>, 4=?, Index=&SpawnGroup(), Num=::NewEnumerator2)
    Sprite Size Mode=::NewEnumerator4
    Write Material Random=true
   )
   &CameraOffset(Camera Offset Amount=&MergeHelper_FloatSwitch(0=-10, 1=-10, 2=0, 3=0, Index=&SpawnGroup(), Num=::NewEnumerator2))
   /SetVariables(Particles.VisibilityTag=&MergeHelper_IntSwitch(0=0, 1=0, 2=1, 3=1, 4=?, 5=?, Index=&SpawnGroup(), Num=::NewEnumerator2)) {isSetParameters=true}
  ]
  particleUpdateScript [
   &ParticleState()
   &SkeletalMeshLocation(
    Bone / Socket Index=&MergeHelper_IntSwitch(0=0, 1=1, 2=0, 3=1, 4=?, 5=?, Index=&SpawnGroup(), Num=::NewEnumerator2)
    Bone Sampling Mode=::NewEnumerator4
    MaskSpawnGroup=false
    Sampled Position Offset=&MergeHelper_VectorSwitch(0=<0 0 46>, 1=<0 0 53>, 2=<0 -0.7 47>, 3=<0 0 53>, Index=&SpawnGroup(), Num=::NewEnumerator2)
    Skeletal Mesh=$System.Skeletal Mesh
   )
   &SpriteFacingAndAlignment(
    Sprite Alignment=&RotateVectorByQuaternion(Vector To Rotate=<0 0 1>, Quaternion=$Particles.SkeletalMeshLocation.SampledRotation)
    Sprite Facing=<0.7068 0.07 0.3>
    WriteAlignment=true
    WriteFacing=false
   )
   &ScaleColor(
    Scale Alpha=&FloatFromCurve(FloatCurve=DataInterface() {propertyValues {Curve {keys [<at 0 0 RCIM_Cubic>,<at 0.5 1 RCIM_Cubic>]}}})
    ScaleA=true
    ScaleRGB=true
   )
  ]
  rendererClasses [&NiagaraSpriteRendererProperties]
  renderers [
   {rendererClass=&NiagaraSpriteRendererProperties}
   {rendererClass=&NiagaraSpriteRendererProperties, rendererIndex=1}
  ]
 }
 EnergyWave {
  emitterUpdateScript [
   &EmitterState(Inactive Response=::NewEnumerator1, Life Cycle Mode=::NewEnumerator1, Loop Behavior=::NewEnumerator1, Loop Duration=2, UseLoopDelay=false)
   &SpawnBurst_Instantaneous(Use Spawn Probability=false, UseLoopCountLimit=false)
   &SpawnBurst_Instantaneous(Spawn Group=1, Spawn Time=0.64, Use Spawn Probability=false, UseLoopCountLimit=false) {module=SpawnBurst_Instantaneous001}
   /SetVariables(
    Emitter.DepthFade:NiagaraFloat=&FN_FrontEndFloatModifier(Gameplay=500, Front End=300)
    Emitter.Refraction:NiagaraFloat=&FN_FrontEndFloatModifier(
     Gameplay=&FN_PickFloatDependOnPlatform(TargetPlatformActive=10, TargetPlatformNoActive=0)
     Front End=0
    )
   ) {isSetParameters=true}
  ]
  particleSpawnScript [
   &InitializeParticle(
    Color Mode=::NewEnumerator0
    Lifetime=&MergeHelper_FloatSwitch(0=1.15, 1=0.3, 2=?, 3=?, Index=&SpawnGroup(), Num=::NewEnumerator0)
    Mesh Scale Mode=::NewEnumerator3
    Mesh Uniform Scale=3
    UsePositionOffset=false
    Write Material Random=true
   )
   &CameraOffset()
   /SetVariables(Particles.LightPosOffset:NiagaraPosition=<10 10 0>) {isSetParameters=true}
   &InitialMeshOrientation(Mesh Axis To Orient=<0 0 1>, Mesh Orientation Mode=::NewEnumerator2, Use Rotation Vector=false)
   /SetVariables(Particles.VisibilityTag=&SpawnGroup()) {isSetParameters=true}
  ]
  particleUpdateScript [
   &ParticleState()
   &ScaleColor(
    Linear Color Curve=DataInterface() {propertyValues {
      AlphaCurve {keys [<at 0 1>,<at 0.2388 1>,<at 1 0>]}
      BlueCurve {keys [<at 0 0>,<at 0.168 0>,<at 0.365 0.0728>,<at 1 0.0089>]}
      GreenCurve {keys [<at 0 0.2181>,<at 0.168 0.3824>,<at 0.365 0.395>,<at 1 0.1141>]}
      RedCurve {keys [<at 0 1>,<at 0.168 1>,<at 0.365 1>,<at 1 1>]}
     }}
    Scale Mode=::NewEnumerator2
   )
   &SkeletalMeshLocation(
    Bone / Socket Index=&MergeHelper_IntSwitch(0=2, 1=0, 2=?, 3=?, 4=?, 5=?, Index=&SpawnGroup(), Num=::NewEnumerator0)
    Bone Sampling Mode=::NewEnumerator4
    MaskSpawnGroup=false
    Orientation Sampling=::NewEnumerator1
    Sampled Position Offset=&MergeHelper_VectorSwitch(0=<0 70 0>, 1=<0 0 37>, 2=?, 3=?, Index=&SpawnGroup(), Num=::NewEnumerator0)
    Skeletal Mesh=$System.Skeletal Mesh
   )
   &DynamicMaterialParameters(
    Index 0 Param 1=&FloatFromCurve(FloatCurve=DataInterface() {propertyValues {Curve {keys [{arriveTangent=-2.1015, interpMode=RCIM_Cubic, leaveTangent=-2.1015, tangentMode=RCTM_User, time=0, value=0.5}, <at 1 -0.15>]}}})
    Index 0 Param 2=&FloatFromCurve(FloatCurve=DataInterface() {propertyValues {Curve {keys [{arriveTangent=-0.3464, interpMode=RCIM_Cubic, leaveTangent=-0.3464, tangentMode=RCTM_User, time=0, value=1}, <at 1 0.9>]}}})
    Index 0 Param 3=&FloatFromCurve(FloatCurve=DataInterface() {propertyValues {Curve {keys [
         {interpMode=RCIM_Cubic, tangentMode=RCTM_Break, time=0, value=0}
         {arriveTangent=0.4478, interpMode=RCIM_Cubic, leaveTangent=0.4478, tangentMode=RCTM_User, time=0.15, value=1}
         {interpMode=RCIM_Cubic, tangentMode=RCTM_Break, time=1, value=1.2}
        ]}}})
    Module.Param0WriteEnabled=true
    Module.Param1WriteEnabled=true
    Module.Param2WriteEnabled=true
    Module.Param3WriteEnabled=false
   )
   &ScaleColor(
    Scale Alpha=&FloatFromCurve(
     FloatCurve=DataInterface() {propertyValues {Curve {keys [{interpMode=RCIM_Cubic, tangentMode=RCTM_User, time=0.5, value=1}, {interpMode=RCIM_Cubic, tangentMode=RCTM_User, time=1, value=0}]}}}
     Scale Curve=&FN_FrontEndFloatModifier(Gameplay=1, Front End=0.5)
    )
    Scale RGB=&VectorFromFloat(Value=&FloatFromCurve(FloatCurve=DataInterface() {propertyValues {Curve {keys [
          {interpMode=RCIM_Cubic, tangentMode=RCTM_Break, time=0, value=6}
          {arriveTangent=-6.6072, interpMode=RCIM_Cubic, leaveTangent=-6.6072, tangentMode=RCTM_User, time=0.15, value=3}
          {interpMode=RCIM_Cubic, tangentMode=RCTM_Break, time=1, value=1}
         ]}}}))
    ScaleA=true
    ScaleRGB=true
   ) {module=ScaleColor001}
   &ScaleMeshSize(Scale Factor=&VectorFromFloat(Value=&FloatFromCurve(
      FloatCurve=DataInterface() {propertyValues {Curve {keys [
          {interpMode=RCIM_Cubic, tangentMode=RCTM_Break, time=0, value=0.2}
          {arriveTangent=2.9365, interpMode=RCIM_Cubic, leaveTangent=2.9365, tangentMode=RCTM_User, time=0.1, value=1}
          {interpMode=RCIM_Cubic, tangentMode=RCTM_Break, time=1, value=2}
         ]}}}
      Scale Curve=0.2
     )))
   &FN_Light_Attributes(
    Color=<rgba 3 0.5742 0.044 1>
    Create Color Attribute=true
    Create Exponent Attribute=true
    Create Position Attribute=true
    Create Radius Attribute=true
    Exponent=1.3
    Position=&Add_Vector(A=Linked() {linkedVariable {name=Particles.Initial.Position, type=NiagaraPosition}}, B=<0 0 0>)
    Radius=100
    Scale Color=&FloatFromCurve(
     FloatCurve=DataInterface() {propertyValues {Curve {keys [
         <at 0 0 RCIM_Cubic>
         {arriveTangent=0.7715, interpMode=RCIM_Cubic, leaveTangent=0.7715, time=0.2, value=0.2}
         {arriveTangent=0.1239, interpMode=RCIM_Cubic, leaveTangent=0.1239, tangentMode=RCTM_User, time=0.5311, value=0.4097}
         {arriveTangent=-0.8738, interpMode=RCIM_Cubic, leaveTangent=-0.8738, time=0.6292, value=1}
         <at 1 0 RCIM_Cubic>
        ]}}}
     Scale Curve=&Multiply_Float(A=15, B=1)
    )
   )
   /SetVariables(Particles.LightPos:NiagaraPosition=&Add_Vector(
     A=Linked() {linkedVariable {name=Particles.Position, type=NiagaraPosition}}
     B=Linked() {linkedVariable {name=Particles.LightPosOffset, type=NiagaraPosition}}
    )) {isSetParameters=true}
  ]
  rendererClasses [&NiagaraMeshRendererProperties, &NiagaraLightRendererProperties]
  renderers [
   {rendererClass=&NiagaraMeshRendererProperties, rendererIndex=0}
   {rendererClass=&NiagaraLightRendererProperties, rendererIndex=1}
  ]
 }
 Flare {
  emitterUpdateScript [
   &EmitterState(Inactive Response=::NewEnumerator1, Life Cycle Mode=::NewEnumerator1, Loop Behavior=::NewEnumerator1, Loop Duration=1, UseLoopDelay=false)
   &SpawnBurst_Instantaneous(Spawn Time=0.64, Use Spawn Probability=false, UseLoopCountLimit=false)
   &SpawnBurst_Instantaneous(Spawn Group=1, Spawn Time=0.64, Use Spawn Probability=false, UseLoopCountLimit=false) {module=SpawnBurst_Instantaneous001}
   &SpawnBurst_Instantaneous(Spawn Group=2, Spawn Time=0.64, Use Spawn Probability=false, UseLoopCountLimit=false) {module=SpawnBurst_Instantaneous002}
  ]
  particleSpawnScript [
   /SetVariables(Particles.VisibilityTag=&SpawnGroup()) {isSetParameters=true}
   &InitializeParticle(
    Color=&SelectLinearColorFromArray(
     Array Sampling Mode=::NewEnumerator1
     Direct Array Index=&SpawnGroup()
     Color Selection Array=DataInterface() {propertyValues {ColorData [<rgba 9 6.0429 3.0723 1>, <rgba 10 3.7287 1.995 1>, <rgba 0 0 0 1>, <rgba 10 4.3313 1.9043 0.7>, <rgba 0.2 0.4 1 0.7>]}}
     Random Seed=?
    )
    Lifetime=&SelectFloatFromArray(
     Direct Array Index=&SpawnGroup()
     Float Selection Array=DataInterface() {propertyValues {FloatData [0.3,0.25,0.3,0.2,0.9]}}
    )
    Mesh Scale Mode=::NewEnumerator3
    Mesh Uniform Scale=7
    Position Offset Coordinate Space=::Local
    Sprite Rotation Mode=::NewEnumerator2
    Sprite Size Mode=::NewEnumerator3
    Uniform Sprite Size=&SelectFloatFromArray(
     Direct Array Index=&SpawnGroup()
     Float Selection Array=DataInterface() {propertyValues {FloatData [120,180,120,0,0]}}
    )
    UsePositionOffset=true
    Write Material Random=true
   )
   &DynamicMaterialParameters(
    Index 0 Param 1=1
    Index 0 Param 2=1
    Index 0 Param 3=1
    Index 0 Param 4=1
    Module.Index0Float0Write=true
    Module.Index0Float1Write=true
    Module.Index0Float2Write=true
    Module.Index0Float3Write=true
   )
  ]
  particleUpdateScript [
   &ParticleState()
   &ScaleSpriteSize(
    Scale Sprite Size Mode=::NewEnumerator0
    Uniform Scale Factor=&FN_FrontEndFloatModifier(
     Gameplay=1
     Front End=&MergeHelper_FloatSwitch(0=0.7, 1=0.3, 2=0.3, 3=1, 4=1, Index=&SpawnGroup(), Num=::NewEnumerator3)
    )
   )
   &ScaleSpriteSize(
    Scale Sprite Size Mode=::NewEnumerator0
    Uniform Scale Factor=&MergeHelper_FloatSwitch(
     0=&Multiply_Float(
      A=1
      B=&FloatFromCurve(
       FloatCurve=DataInterface() {propertyValues {Curve {keys [
           {arriveTangent=2.6083, interpMode=RCIM_Cubic, leaveTangent=0.4318, tangentMode=RCTM_User, time=0, value=0.5}
           {arriveTangent=1.3014, interpMode=RCIM_Cubic, leaveTangent=1.3014, tangentMode=RCTM_User, time=0.3, value=0.8}
           {arriveTangent=0.1331, leaveTangent=0.1331, time=1, value=1.2}
          ]}}}
       Scale Curve=1.25
      )
     )
     1=&Multiply_Float(
      A=1
      B=&FloatFromCurve(
       FloatCurve=DataInterface() {propertyValues {Curve {keys [
           {arriveTangent=2.6083, interpMode=RCIM_Cubic, leaveTangent=0.4318, tangentMode=RCTM_User, time=0, value=0.5}
           {arriveTangent=1.3014, interpMode=RCIM_Cubic, leaveTangent=1.3014, tangentMode=RCTM_User, time=0.3, value=0.8}
           {arriveTangent=0.1331, leaveTangent=0.1331, time=1, value=1.2}
          ]}}}
       Scale Curve=1.25
      )
     )
     2=&Multiply_Float(
      A=1
      B=&FloatFromCurve(
       FloatCurve=DataInterface() {propertyValues {Curve {keys [
           {arriveTangent=2.6083, interpMode=RCIM_Cubic, leaveTangent=0.4318, tangentMode=RCTM_User, time=0, value=0.5}
           {arriveTangent=1.3014, interpMode=RCIM_Cubic, leaveTangent=1.3014, tangentMode=RCTM_User, time=0.3, value=0.8}
           {arriveTangent=0.1331, leaveTangent=0.1331, time=1, value=1.2}
          ]}}}
       Scale Curve=1.25
      )
     )
     3=1
     Index=&SpawnGroup()
     Num=::NewEnumerator2
    )
   ) {module=ScaleSpriteSize002}
   &CameraOffset(Camera Offset Amount=&MergeHelper_FloatSwitch(0=20, 1=50, 2=20, 3=20, 4=50, Index=&SpawnGroup(), Num=::NewEnumerator3))
   &ScaleColor(
    Scale Alpha=&MaskFloatBySpawnGroup(
     Masked Float=&FloatFromCurve(FloatCurve=DataInterface() {propertyValues {Curve {keys [{arriveTangent=-3.3102, interpMode=RCIM_Cubic, time=0, value=0}, <at 0.5 1 RCIM_Cubic>, <at 1 0 RCIM_Cubic>]}}})
     Passthrough Float=&MaskFloatBySpawnGroup(
      Masked Float=0.2
      Passthrough Float=&MaskFloatBySpawnGroup(
       Masked Float=&FloatFromCurve(FloatCurve=DataInterface() {propertyValues {Curve {keys [{interpMode=RCIM_Cubic, tangentMode=RCTM_User, time=0, value=1}, {interpMode=RCIM_Cubic, tangentMode=RCTM_User, time=1, value=0}]}}})
       Passthrough Float=&FloatFromCurve(FloatCurve=DataInterface() {propertyValues {Curve {keys [
            {interpMode=RCIM_Cubic, tangentMode=RCTM_User, time=0, value=1}
            {interpMode=RCIM_Cubic, tangentMode=RCTM_User, time=0.1, value=1}
            {interpMode=RCIM_Cubic, tangentMode=RCTM_User, time=1, value=0}
           ]}}})
       Spawn Group Mask=2
      )
      Spawn Group Mask=1
     )
    )
    Scale RGB=&MaskVectorBySpawnGroup(
     Masked Vector=&MakeVectorFromLinearColorRGB(Color=&ColorFromCurve(ColorCurve=DataInterface() {propertyValues {
         AlphaCurve {keys []}
         BlueCurve {keys [<at 0 1.8257 RCIM_Cubic>,<at 0.3877 0>,<at 1 0.0229 RCIM_Cubic>]}
         GreenCurve {keys [<at 0 2.8793 RCIM_Cubic>,<at 0.3877 0.225>,<at 1 0.0716 RCIM_Cubic>]}
         RedCurve {keys [<at 0 10 RCIM_Cubic>,<at 0.3877 1>,<at 1 0.1878 RCIM_Cubic>]}
        }}))
     Passthrough Vector=&MaskVectorBySpawnGroup(
      Masked Vector=&VectorFromFloat(Value=&FloatFromCurve(FloatCurve=DataInterface() {propertyValues {Curve {keys [{interpMode=RCIM_Cubic, tangentMode=RCTM_User, time=0, value=1}, {interpMode=RCIM_Cubic, tangentMode=RCTM_User, time=1, value=0}]}}}))
      Passthrough Vector=<1 1 1>
      Spawn Group Mask=1
     )
    )
    ScaleA=true
    ScaleRGB=true
   )
   &ScaleColor(
    Scale Alpha=&Multiply_Float(
     A=&FN_FrontEndFloatModifier(
      Gameplay=1
      Front End=&MergeHelper_FloatSwitch(
       0=&FN_PickFloatDependOnPlatform(TargetPlatform=DataInterface() {propertyValues {Platforms {qualityLevelMask=-29}}}, TargetPlatformActive=0.8, TargetPlatformNoActive=0.4)
       1=&FN_PickFloatDependOnPlatform(TargetPlatform=DataInterface() {propertyValues {Platforms {qualityLevelMask=-29}}}, TargetPlatformActive=0.8, TargetPlatformNoActive=0.4)
       2=1
       3=0.222
       4=1
       Index=&SpawnGroup()
       Num=::NewEnumerator3
      )
     )
     B=&FN_PickFloatDependOnPlatform(TargetPlatform=DataInterface() {propertyValues {Platforms {qualityLevelMask=-29}}}, TargetPlatformActive=0.2, TargetPlatformNoActive=1)
    )
    Scale RGB=&VectorFromFloat(Value=&FN_PickFloatDependOnPlatform(TargetPlatform=DataInterface() {propertyValues {Platforms {qualityLevelMask=-29}}}, TargetPlatformActive=0.4, TargetPlatformNoActive=1))
    ScaleA=true
    ScaleRGB=true
   ) {module=ScaleColor001}
   &SkeletalMeshLocation(Bone Sampling Mode=::NewEnumerator4, MaskSpawnGroup=false, Sampled Position Offset=<0 0 37>, Skeletal Mesh=$System.Skeletal Mesh)
  ]
  rendererClasses [&NiagaraSpriteRendererProperties]
  renderers [
   {rendererClass=&NiagaraSpriteRendererProperties}
   {rendererClass=&NiagaraSpriteRendererProperties, rendererIndex=1}
   {rendererClass=&NiagaraSpriteRendererProperties, rendererIndex=2}
  ]
 }
 LightiningBurst {
  emitterUpdateScript [
   &EmitterState(Inactive Response=::NewEnumerator1, Life Cycle Mode=::NewEnumerator1, Loop Behavior=::NewEnumerator1, Loop Duration=1, UseLoopDelay=false)
   &SpawnBurst_Instantaneous(Spawn Count=5, Spawn Time=0.6, Use Spawn Probability=false, UseLoopCountLimit=false)
  ]
  particleSpawnScript [
   &InitializeParticle(
    Color=&MaskLinearColorBySpawnGroup(
     Masked Linear Color=&RandomRangeLinearColor(Minimum=<rgba 0.6 0.6 0.6 0.5>, Maximum=<rgba 0.6 0.6 0.6 1>)
     Passthrough Linear Color=&RandomRangeLinearColor(Minimum=<rgba 0.8 0.8 0.8 0.5>, Maximum=<rgba 1 1 1 1>)
    )
    Lifetime=&RandomRangeFloat(Minimum=0.2, Maximum=0.3)
    Mesh Scale Mode=::NewEnumerator3
    Mesh Uniform Scale=&MaskFloatBySpawnGroup(
     Masked Float=&RandomRangeFloat(Minimum=3.5, Maximum=3.5)
     Passthrough Float=&RandomRangeFloat(Minimum=4, Maximum=4)
    )
    UsePositionOffset=false
    Write Material Random=true
   )
   &AddVelocity(Velocity=&RandomRangeVector(Minimum=<-7 -10 50>, Maximum=<7 50 80>))
   &InitialMeshOrientation(Mesh Axis To Orient=$Particles.Velocity, Mesh Orientation Mode=::NewEnumerator2, Orientation Vector=<0 0 1>, Use Rotation Vector=false)
  ]
  particleUpdateScript [
   &ParticleState()
   &SkeletalMeshLocation(
    Bone / Socket Index=&SpawnGroup()
    Bone Sampling Mode=::NewEnumerator4
    MaskSpawnGroup=false
    Sampled Position Offset=&RandomRangeVector(Minimum=<0 0 20>, Maximum=<0 0 90>)
    Skeletal Mesh=$System.Skeletal Mesh
   ) {module=SkeletalMeshLocation001}
   &ScaleMeshSize(Scale Factor=&RandomRangeVector(Minimum=<1 1 2>, Maximum=<2 1.5 2>)) {module=ScaleMeshSize001}
   &ScaleMeshSize(Scale Factor=&MakeVector(
     X=1
     Y=1
     Z=&FloatFromCurve(FloatCurve=DataInterface() {propertyValues {Curve {keys [
          {interpMode=RCIM_Cubic, leaveTangent=0.0966, tangentMode=RCTM_Break, time=0, value=1}
          {arriveTangent=0.0951, interpMode=RCIM_Cubic, leaveTangent=-0.1745, tangentMode=RCTM_Break, time=1, value=1.07}
         ]}}})
    ))
   &ScaleMeshSize(Scale Factor=&VectorFromFloat(Value=&RandomRangeFloat(Minimum=0.03, Maximum=0.05))) {module=ScaleMeshSize002}
   &DynamicMaterialParameters(
    Index 0 Param 4=&SelectFloatFromArray(
     Direct Array Index=&ModuloInt(A=&ReturnParticleIDAttributes(), B=4)
     Float Selection Array=DataInterface() {propertyValues {FloatData [3,0,10,3]}}
    )
    Index 1 Param 1=1.2
    Index 1 Param 2=2
    Index 1 Param 3=&FloatFromCurve(FloatCurve=DataInterface() {propertyValues {Curve {keys [{interpMode=RCIM_Cubic, tangentMode=RCTM_Break, time=0.5, value=-0.5}, <at 1 0.1>]}}})
    Index 2 Param 1=&Multiply_Float(
     A=&FloatFromCurve(
      FloatCurve=DataInterface() {propertyValues {Curve {keys [{interpMode=RCIM_Cubic, leaveTangent=8.1085, tangentMode=RCTM_Break, time=0.8568, value=1}, {tangentMode=RCTM_User, time=1, value=0}]}}}
      Scale Curve=0.35
     )
     B=&Divide_Float(
      A=&MakeFloatFromVector(Vector=$Particles.Scale, Channel=::NewEnumerator0)
      B=&RandomRangeFloat(Minimum=10, Maximum=15)
     )
    )
    Index 2 Param 2=&FloatFromCurve(
     FloatCurve=DataInterface() {propertyValues {Curve {keys [
         {interpMode=RCIM_Cubic, tangentMode=RCTM_Break, time=0.7, value=0}
         {arriveTangent=-0.0859, interpMode=RCIM_Cubic, tangentMode=RCTM_Break, time=1, value=-1}
        ]}}}
     Scale Curve=50
    )
    Index 2 Param 3=&Waveform(
     Global Input Position=&RandomRangeFloat(Minimum=0, Maximum=155)
     Global Phase=$Particles.Age
     "Waveform [1]"=::NewEnumerator12
     "[1] Phase Scale"=2
     "[1] Frequency"=1
     "[1] Amplitude Min/Max"=<0 1>
    )
    Index 2 Param 4=25
    Index 3 Param 1=&FloatFromCurve(
     FloatCurve=DataInterface() {propertyValues {Curve {keys [
         {interpMode=RCIM_Cubic, leaveTangent=-0.0025, tangentMode=RCTM_Break, time=0, value=0}
         {arriveTangent=1.8528, interpMode=RCIM_Cubic, tangentMode=RCTM_Break, time=1, value=1}
        ]}}}
     Scale Curve=&RandomRangeFloat(Minimum=15, Maximum=25)
    )
    Index 3 Param 2=&FloatFromCurve(FloatCurve=DataInterface() {propertyValues {Curve {keys [
         {interpMode=RCIM_Cubic, leaveTangent=-0.0025, tangentMode=RCTM_Break, time=0, value=0.5}
         {arriveTangent=-0.0316, interpMode=RCIM_Cubic, tangentMode=RCTM_Break, time=1, value=-0.2}
        ]}}})
    Index 3 Param 3=&FloatFromCurve(FloatCurve=DataInterface() {propertyValues {Curve {keys [
         {interpMode=RCIM_Cubic, leaveTangent=1.1034, tangentMode=RCTM_Break, time=0, value=-1}
         {arriveTangent=2.2399, tangentMode=RCTM_User, time=0.5, value=0.125}
        ]}}})
    Index 3 Param 4=&FloatFromCurve(FloatCurve=DataInterface() {propertyValues {Curve {keys [
         {interpMode=RCIM_Cubic, leaveTangent=3.0601, tangentMode=RCTM_Break, time=0.5, value=-0.5}
         {arriveTangent=3.2053, tangentMode=RCTM_User, time=1, value=1}
        ]}}})
    Module.Index0Param0Write=true
    Module.Index0Param1Write=true
    Module.Index0Param2Write=true
    Module.Index0Param3Write=true
    Module.Index1Param0Write=true
    Module.Index1Param1Write=true
    Module.Index1Param2Write=true
    Module.Index1Param3Write=true
    Module.Index2Param0Write=true
    Module.Index2Param1Write=true
    Module.Index2Param2Write=true
    Module.Index2Param3Write=true
    Module.Index3Param0Write=true
    Module.Index3Param1Write=true
    Module.Index3Param2Write=true
    Module.Index3Param3Write=true
    Write Parameter Index 1=true
    Write Parameter Index 2=true
    Write Parameter Index 3=true
   )
   &SolveForcesAndVelocity(Clamp Velocity=false, Limit Acceleration=false)
   &ScaleColor(
    Scale Alpha=&FloatFromCurve(FloatCurve=DataInterface() {propertyValues {Curve {keys [
         {tangentMode=RCTM_User, time=0, value=1}
         {arriveTangent=-0.0405, interpMode=RCIM_Cubic, leaveTangent=-0.0405, tangentMode=RCTM_User, time=0.7009, value=1}
         {arriveTangent=-8.0089, interpMode=RCIM_Cubic, leaveTangent=-8.0089, tangentMode=RCTM_User, time=1, value=0}
        ]}}})
    Scale RGB=&Lerp_Vector(
     A=&VectorFromFloat(Value=6)
     B=&VectorFromFloat(Value=&RandomRangeFloat(Minimum=1, Maximum=1.5, Evaluation Type=::NewEnumerator1))
     Alpha=&FloatFromCurve(FloatCurve=DataInterface() {propertyValues {Curve {keys [{interpMode=RCIM_Cubic, tangentMode=RCTM_Break, time=0, value=0}, {interpMode=RCIM_Cubic, tangentMode=RCTM_Break, time=0.333, value=1}]}}})
    )
    ScaleA=true
    ScaleRGB=true
   )
  ]
  rendererClasses [&NiagaraMeshRendererProperties]
  renderers [
   {rendererClass=&NiagaraMeshRendererProperties, rendererIndex=0}
   {rendererClass=&NiagaraMeshRendererProperties, rendererIndex=1}
  ]
 }
}
summary {
 systemName=NS_Emote_SwordSplit_02
 emitters [
  {emitterName=ConnectingArcs, rendererClasses [&NiagaraMeshRendererProperties, &NiagaraSpriteRendererProperties]}
  {emitterName=EnergyWave, rendererClasses [&NiagaraMeshRendererProperties, &NiagaraLightRendererProperties]}
  {emitterName=LightiningBurst, rendererClasses [&NiagaraMeshRendererProperties]}
  {emitterName=EdgeGlow, rendererClasses [&NiagaraSpriteRendererProperties]}
  {emitterName=Flare, rendererClasses [&NiagaraSpriteRendererProperties]}
 ]
 userVariables [
  User.bIsFrontEnd {type=NiagaraBool}
  User.bIsFrontEndPreview {type=NiagaraBool}
 ]
}
