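# Niagara effect for the anticipation phase of the SwordSplit emote, crackling orange lightning along two skeletal mesh bones or sockets.
# LightningArcs spawns two overlapping waves from System.Age curves, while LightningStreaks spawns only during the first.
# Spawn groups select bone, offset and scale, and particles resample the skeleton each update to follow the animation.
# Mesh and sprite renderers use animated material parameters, with front-end dimming and, for arcs, time-of-day dimming.
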NiagaraSystem /BRCosmetics/Animation/Game/MainPlayer/Emotes/SwordSplit/FX/NS_Emote_SwordSplit_Anticipation_Lightning_02
dependencies {
 usedDataInterfaces [&NiagaraDataInterfaceCurve, &NiagaraDataInterfaceSkeletalMesh]
 usedDynamicInputs []
 usedModules [
  &self:SystemSpawnScript.SystemScriptSource.SystemScriptGraph.NiagaraNodeAssignment_0.SetVariables_#1
  &SystemState
 ]
}
emitters {
 LightningArcs {
  emitterUpdateScript [
   &EmitterState(Inactive Response=::NewEnumerator1, Life Cycle Mode=::NewEnumerator1, Loop Behavior=::NewEnumerator1, UseLoopDelay=false)
   &SpawnRate(
    SpawnRate=&Multiply_Float(
     A=2.5
     B=&FloatFromCurve(
      FloatCurve=DataInterface() {propertyValues {Curve {keys [<at 0 0 RCIM_Cubic>, <at 0.9239 1 RCIM_Cubic>, {arriveTangent=-11.1136, tangentMode=RCTM_User, time=1.2, value=0}]}}}
      CurveIndex=$System.Age
     )
    )
    Use Spawn Probability=false
   )
   &SpawnRate(
    SpawnGroup=1
    SpawnRate=&Multiply_Float(
     A=2.5
     B=&FloatFromCurve(
      FloatCurve=DataInterface() {propertyValues {Curve {keys [<at 0 0 RCIM_Cubic>, <at 0.9239 1 RCIM_Cubic>, {arriveTangent=-11.1136, tangentMode=RCTM_User, time=1.2, value=0}]}}}
      CurveIndex=$System.Age
     )
    )
    Use Spawn Probability=false
   ) {module=SpawnRate002}
   &SpawnRate(
    SpawnGroup=2
    SpawnRate=&Multiply_Float(
     A=7
     B=&FloatFromCurve(
      FloatCurve=DataInterface() {propertyValues {Curve {keys [
          {arriveTangent=0.1296, interpMode=RCIM_Cubic, time=0.85, value=0}
          {arriveTangent=1.8541, interpMode=RCIM_Cubic, leaveTangent=1.8541, tangentMode=RCTM_User, time=0.95, value=0.9}
          {arriveTangent=-0.7169, interpMode=RCIM_Cubic, leaveTangent=-0.7169, time=1.7213, value=1}
          <at 2.2055 0>
         ]}}}
      CurveIndex=$System.Age
     )
    )
    Use Spawn Probability=false
   ) {module=SpawnRate001}
   &SpawnRate(
    SpawnGroup=3
    SpawnRate=&Multiply_Float(
     A=7
     B=&FloatFromCurve(
      FloatCurve=DataInterface() {propertyValues {Curve {keys [
          {arriveTangent=0.1296, interpMode=RCIM_Cubic, time=0.85, value=0}
          {arriveTangent=1.8541, interpMode=RCIM_Cubic, leaveTangent=1.8541, tangentMode=RCTM_User, time=0.95, value=0.9}
          {arriveTangent=-0.7169, interpMode=RCIM_Cubic, leaveTangent=-0.7169, time=1.7213, value=1}
          <at 2.2055 0>
         ]}}}
      CurveIndex=$System.Age
     )
    )
    Use Spawn Probability=false
   ) {module=SpawnRate003}
  ]
  particleSpawnScript [
   /SetVariables(Particles.RandFloat:NiagaraFloat=&UniformAOrBFloat(
     A=&RandomRangeFloat(Minimum=0, Maximum=0.45)
     B=&RandomRangeFloat(Minimum=0.45, Maximum=1)
     Evaluation Type=::NewEnumerator0
     Recalculate Random Each Loop=true
     Override Seed=false
    )) {isSetParameters=true}
   &InitializeParticle(
    Color=<rgba 1 0.1802 0.0537 1>
    Lifetime=&MergeHelper_FloatSwitch(
     0=&RandomRangeFloat(Minimum=0.2, Maximum=0.4)
     1=&RandomRangeFloat(Minimum=0.2, Maximum=0.4)
     2=&RandomRangeFloat(Minimum=0.2, Maximum=0.4)
     3=&RandomRangeFloat(Minimum=0.2, Maximum=0.4)
     Index=&SpawnGroup()
     Num=::NewEnumerator2
    )
    Mesh Scale=&Multiply_Vector(
     A=&RandomRangeVector(Minimum=<0.5 1 1>, Maximum=<3 1 4>)
     B=&VectorFromFloat(Value=&MergeHelper_FloatSwitch(0=1, 1=1, 2=1, 3=1, Index=&SpawnGroup(), Num=::NewEnumerator2))
    )
    Mesh Scale Mode=::NewEnumerator4
    UsePositionOffset=false
    Write Material Random=true
   )
   &CameraOffset(Camera Offset Amount=14)
   &InitialMeshOrientation(Rotation=&RandomRangeVector(Minimum=<-1 -1 -1>, Maximum=<1 1 1>), Use Rotation Vector=true) {module=InitialMeshOrientation002}
  ]
  particleUpdateScript [
   &ParticleState()
   &ScaleMeshSize(Scale Factor=&Multiply_VectorByFloat(
     Vector=<1 1 1>
     Float=&Multiply_Float(
      A=&MergeHelper_FloatSwitch(0=0.4, 1=0.4, 2=0.3, 3=0.3, Index=&SpawnGroup(), Num=::NewEnumerator2)
      B=1
     )
    ))
   &ScaleColor(
    Scale Alpha=&Multiply_Float(
     A=10
     B=&FloatFromCurve(FloatCurve=DataInterface() {propertyValues {Curve {keys [<at 0 0>,<at 0.1 0>,<at 0.2 0.8>,<at 1 0>]}}})
    )
    Scale RGB=&Multiply_VectorByFloat(
     Vector=<1 1 1>
     Float=&Multiply_Float(
      A=&RandomRangeFloat(Minimum=0, Maximum=1, Evaluation Type=::NewEnumerator1)
      B=&Multiply_Float(
       A=1
       B=&FloatFromCurve(
        FloatCurve=DataInterface() {propertyValues {Curve {keys [<at 0 10>,<at 1 10>]}}}
        Scale Curve=&FNTimeOfDayScalar(Night Scale=0.5, Morning Scale=0.5, Evening Scale=0.5)
       )
      )
     )
    )
    ScaleA=true
    ScaleRGB=true
   )
   &SkeletalMeshLocation(
    Bone / Socket Index=&MergeHelper_IntSwitch(0=0, 1=1, 2=0, 3=1, 4=?, 5=?, Index=&SpawnGroup(), Num=::NewEnumerator2)
    Bone Sampling Mode=::NewEnumerator4
    MaskSpawnGroup=false
    Sampled Position Offset=&MergeHelper_VectorSwitch(
     0=&Lerp_Vector(A=<0 0 10>, B=<1 1 40>, Alpha=$Particles.RandFloat)
     1=&Lerp_Vector(A=<0 0 10>, B=<1 1 40>, Alpha=$Particles.RandFloat)
     2=&Lerp_Vector(A=<0 0 -15>, B=<1 1 -5>, Alpha=$Particles.RandFloat)
     3=&Lerp_Vector(A=<0 0 -15>, B=<1 1 -5>, Alpha=$Particles.RandFloat)
     Index=&SpawnGroup()
     Num=::NewEnumerator2
    )
    Skeletal Mesh=$System.Skeletal Mesh
   )
   &DynamicMaterialParameters(
    Index 0 Param 1=&FloatFromCurve(FloatCurve=DataInterface() {propertyValues {Curve {keys [{tangentMode=RCTM_User, time=0, value=0.2}, {tangentMode=RCTM_User, time=1, value=1}]}}})
    Index 0 Param 2=0.7
    Index 0 Param 3=&FloatFromCurve(FloatCurve=DataInterface() {propertyValues {Curve {keys [<at 0 0 RCIM_Cubic>, {arriveTangent=-0.3684, interpMode=RCIM_Cubic, leaveTangent=-0.3684, tangentMode=RCTM_User, time=1, value=3}]}}})
    Index 0 Param 4=1
    Index 1 Param 1=&MergeHelper_FloatSwitch(0=0.8, 1=0.8, 2=1, 3=1, Index=&SpawnGroup(), Num=::NewEnumerator2)
    Index 1 Param 3=0.3
    Index 1 Param 4=1
    Index 2 Param 1=&FloatFromCurve(FloatCurve=DataInterface() {propertyValues {Curve {keys [<at 0 1>,<at 1 0.3>]}}})
    Module.Index0Float0Write=true
    Module.Index0Float1Write=true
    Module.Index0Float2Write=true
    Module.Index0Float3Write=true
    Module.Index1Float0Write=true
    Module.Index1Float1Write=true
    Module.Index1Float2Write=true
    Module.Index1Float3Write=true
    Module.Index2Float0Write=true
    Module.Index2Float1Write=true
    Module.Index2Float2Write=true
    Module.Index2Float3Write=true
    Write Parameter Index 1=::NewEnumerator1
    Write Parameter Index 2=::NewEnumerator1
   )
   &SolveForcesAndVelocity(Clamp Velocity=false, Limit Acceleration=false)
   /SetVariables(Particles.Lightning glow color:LinearColor=&MakeLinearColorFromVectorAndFloat(
     "Vector (RGB)"=&MakeVectorFromLinearColorRGB(Color=&MultiplyLinearColor(
       Linear Color=<rgba 1.3 0.2025 0.0406 1>
       Scale Factor=&MakeVector4_FromFloat(Float=&FN_FrontEndFloatModifier(Gameplay=0.5, Front End=0.4))
      ))
     "Float (Alpha)"=&Multiply_Float(
      A=&FloatFromCurve(
       FloatCurve=DataInterface() {propertyValues {Curve {keys [<at 0 1 RCIM_Cubic>, {interpMode=RCIM_Cubic, leaveTangent=-3.6307, time=0.9, value=0}]}}}
       Scale Curve=&FNTimeOfDayScalar(Night Scale=0.5, Morning Scale=0.6, Evening Scale=0.6)
      )
      B=1.5
     )
    )) {isSetParameters=true}
   /SetVariables(Particles.Glow Dy02:Vector4f=&MakeVector4(
     X=&FloatFromCurve(FloatCurve=DataInterface() {propertyValues {Curve {keys [{tangentMode=RCTM_User, time=0, value=0.2}, {tangentMode=RCTM_User, time=0.5, value=1}]}}})
     Y=0.7
     Z=0
     W=1
    )) {isSetParameters=true}
   /SetVariables(Particles.Glow Dy03:Vector4f=&MakeVector4(
     X=&FloatFromCurve(FloatCurve=DataInterface() {propertyValues {Curve {keys [<at 0.5 1>,<at 1 0.3>]}}})
     Y=0
     Z=0.3
     W=1
    )) {isSetParameters=true}
   &FN_Sprite_Attributes(
    Camera Offset=20
    Color=&MakeLinearColorFromVectorAndFloat(
     "Vector (RGB)"=&Multiply_Vector(A=&MakeVectorFromLinearColorRGB(Color=<rgba 1 0.141 0.0295 1>), B=<1 1 1>)
     "Float (Alpha)"=&Multiply_Float(
      A=&FloatFromCurve(FloatCurve=DataInterface() {propertyValues {Curve {keys [
           {interpMode=RCIM_Cubic, tangentMode=RCTM_User, time=0, value=0}
           {interpMode=RCIM_Cubic, tangentMode=RCTM_User, time=0.3, value=1}
           {tangentMode=RCTM_User, time=1, value=0}
          ]}}})
      B=1
     )
    )
    Create Camera Offset Attribute=true
    Create Color Attribute=true
    Create SpriteSize Attribute=true
    SpriteSize=&Multiply_Vector2D_ByFloat(
     Float=&FloatFromCurve(FloatCurve=DataInterface() {propertyValues {Curve {keys [
          {interpMode=RCIM_Cubic, tangentMode=RCTM_User, time=0, value=0}
          {interpMode=RCIM_Cubic, tangentMode=RCTM_User, time=0.3, value=1}
          {tangentMode=RCTM_User, time=1, value=0}
         ]}}})
     Vector2D=<40 40>
    )
    Write to Module Namespace=true
   )
   /SetVariables(Particles.DecalDynamic2:Vector4f=<0 0 1 1>) {isSetParameters=true}
  ]
  rendererClasses [&NiagaraMeshRendererProperties, &NiagaraSpriteRendererProperties]
  renderers [
   {rendererClass=&NiagaraMeshRendererProperties, rendererIndex=0}
   {rendererClass=&NiagaraMeshRendererProperties, rendererIndex=1}
   {rendererClass=&NiagaraSpriteRendererProperties, rendererIndex=2}
  ]
 }
 LightningStreaks {
  emitterUpdateScript [
   &EmitterState(Inactive Response=::NewEnumerator1, Life Cycle Mode=::NewEnumerator1, Loop Behavior=::NewEnumerator1, UseLoopDelay=false)
   &SpawnRate(
    SpawnRate=&Multiply_Float(
     A=8
     B=&FloatFromCurve(
      FloatCurve=DataInterface() {propertyValues {Curve {keys [<at 0 0 RCIM_Cubic>, <at 0.9239 1 RCIM_Cubic>, {arriveTangent=-11.1484, tangentMode=RCTM_User, time=1.2, value=0}]}}}
      CurveIndex=$System.Age
     )
    )
    Use Spawn Probability=false
   )
   &SpawnRate(
    SpawnGroup=1
    SpawnRate=&Multiply_Float(
     A=8
     B=&FloatFromCurve(
      FloatCurve=DataInterface() {propertyValues {Curve {keys [<at 0 0 RCIM_Cubic>, <at 0.9239 1 RCIM_Cubic>, {arriveTangent=-8.1485, tangentMode=RCTM_User, time=1.2, value=0}]}}}
      CurveIndex=$System.Age
     )
    )
    Use Spawn Probability=false
   ) {module=SpawnRate002}
  ]
  particleSpawnScript [
   /SetVariables(Particles.SizeToPosLerp:NiagaraFloat=&RandomRangeFloat(Minimum=0, Maximum=1)) {isSetParameters=true}
   /SetVariables(Particles.RandFloat:NiagaraFloat=&UniformAOrBFloat(
     A=&RandomRangeFloat(Minimum=0, Maximum=0.45)
     B=&RandomRangeFloat(Minimum=0.45, Maximum=1)
     Evaluation Type=::NewEnumerator0
     Recalculate Random Each Loop=true
     Override Seed=false
    )) {isSetParameters=true}
   &InitializeParticle(
    Color=<rgba 1 0.1802 0.0537 1>
    Lifetime=&MergeHelper_FloatSwitch(
     0=&RandomRangeFloat(Minimum=0.2, Maximum=0.4)
     1=&RandomRangeFloat(Minimum=0.2, Maximum=0.4)
     2=&RandomRangeFloat(Minimum=0.2, Maximum=0.3)
     3=&RandomRangeFloat(Minimum=0.2, Maximum=0.3)
     Index=&SpawnGroup()
     Num=::NewEnumerator2
    )
    Mesh Scale=&MergeHelper_VectorSwitch(
     0=&Multiply_Vector(
      A=&RandomRangeVector(Minimum=<1 0.1 0.3>, Maximum=<1 0.1 0.5>)
      B=&VectorFromFloat(Value=&RandomRangeFloat(Minimum=0.8, Maximum=1.4))
     )
     1=&Multiply_Vector(
      A=&RandomRangeVector(Minimum=<1 0.1 0.3>, Maximum=<1 0.1 0.5>)
      B=&VectorFromFloat(Value=&RandomRangeFloat(Minimum=0.8, Maximum=1.4))
     )
     2=&Multiply_Vector(
      A=&RandomRangeVector(Minimum=<0.3 0.3 1>, Maximum=<1.3 1.3 1>)
      B=&VectorFromFloat(Value=&Lerp_Float(A=0.6, B=0.5, Alpha=$Particles.RandFloat))
     )
     3=&Multiply_Vector(
      A=&RandomRangeVector(Minimum=<0.3 0.3 1>, Maximum=<1.3 1.3 1>)
      B=&VectorFromFloat(Value=&Lerp_Float(A=0.6, B=0.5, Alpha=$Particles.RandFloat))
     )
     Index=&SpawnGroup()
     Num=::NewEnumerator2
    )
    Mesh Scale Mode=::NewEnumerator4
    UsePositionOffset=false
    Write Material Random=true
   )
   &CameraOffset(Camera Offset Amount=14)
   /SetVariables(Particles.Random=&RandomRangeFloat(Minimum=0, Maximum=1)) {isSetParameters=true}
   &InitialMeshOrientation(Rotation=&RandomRangeVector(Minimum=<0 0 0>, Maximum=<1 1 1>), Use Rotation Vector=true)
  ]
  particleUpdateScript [
   &ParticleState()
   &ScaleMeshSize(Scale Factor=&Multiply_VectorByFloat(
     Vector=<1 1 1>
     Float=&MergeHelper_FloatSwitch(
      0=&FloatFromCurve(FloatCurve=DataInterface() {propertyValues {Curve {keys [<at 0 1.75>,<at 1 1>]}}}, CurveIndex=$System.Age, Scale Curve=0.36)
      1=&FloatFromCurve(FloatCurve=DataInterface() {propertyValues {Curve {keys [<at 0 1.75>,<at 1 1>]}}}, CurveIndex=$System.Age, Scale Curve=0.36)
      2=&FloatFromCurve(FloatCurve=DataInterface() {propertyValues {Curve {keys [<at 0.7 1.75>,<at 2.7 0.5>]}}}, CurveIndex=$System.Age)
      3=&FloatFromCurve(FloatCurve=DataInterface() {propertyValues {Curve {keys [<at 0.7 1.75>,<at 2.7 0.5>]}}}, CurveIndex=$System.Age)
      Index=&SpawnGroup()
      Num=::NewEnumerator2
     )
    )) {module=ScaleMeshSize001}
   &ScaleMeshSize(Scale Factor=&Multiply_VectorByFloat(
     Vector=<1 0.7 1>
     Float=&Multiply_Float(
      A=&MergeHelper_FloatSwitch(0=1, 1=1, 2=1.3, 3=1.3, Index=&SpawnGroup(), Num=::NewEnumerator2)
      B=1
     )
    ))
   &ScaleColor(
    Scale Alpha=&Multiply_Float(
     A=10
     B=&FloatFromCurve(
      FloatCurve=DataInterface() {propertyValues {Curve {keys [<at 0 0>,<at 0.1 0>,<at 0.2 0.8>,<at 1 0>]}}}
      Scale Curve=&FN_FrontEndFloatModifier(Gameplay=1, Front End=0.7)
     )
    )
    Scale RGB=&Multiply_VectorByFloat(
     Vector=<1 1 1>
     Float=&Multiply_Float(
      A=&RandomRangeFloat(Minimum=0, Maximum=1, Evaluation Type=::NewEnumerator1)
      B=&Multiply_Float(
       A=1
       B=&FloatFromCurve(FloatCurve=DataInterface() {propertyValues {Curve {keys [<at 0 10>,<at 1 10>]}}})
      )
     )
    )
    ScaleA=true
    ScaleRGB=true
   )
   &SkeletalMeshLocation(
    Bone / Socket Index=&MergeHelper_IntSwitch(0=0, 1=1, 2=0, 3=1, 4=?, 5=?, Index=&SpawnGroup(), Num=::NewEnumerator2)
    Bone Sampling Mode=::NewEnumerator4
    MaskSpawnGroup=false
    Sampled Position Offset=&MergeHelper_VectorSwitch(
     0=&Lerp_Vector(A=<0 0 20>, B=<0 0 80>, Alpha=$Particles.SizeToPosLerp)
     1=&Lerp_Vector(A=<0 0 20>, B=<0 0 80>, Alpha=$Particles.SizeToPosLerp)
     2=&Lerp_Vector(A=<0 0 -15>, B=<0 0 70>, Alpha=$Particles.RandFloat)
     3=&Lerp_Vector(A=<0 0 -15>, B=<0 0 70>, Alpha=$Particles.RandFloat)
     Index=&SpawnGroup()
     Num=::NewEnumerator2
    )
    Skeletal Mesh=$System.Skeletal Mesh
   )
   &DynamicMaterialParameters(
    Index 0 Param 1=&FloatFromCurve(FloatCurve=DataInterface() {propertyValues {Curve {keys [{tangentMode=RCTM_User, time=0, value=0.2}, {tangentMode=RCTM_User, time=1, value=1}]}}})
    Index 0 Param 2=0.7
    Index 0 Param 3=&FloatFromCurve(FloatCurve=DataInterface() {propertyValues {Curve {keys [
         {arriveTangent=12.4916, interpMode=RCIM_Cubic, leaveTangent=12.4916, tangentMode=RCTM_User, time=0, value=0}
         {arriveTangent=3, interpMode=RCIM_Cubic, leaveTangent=3, time=0.0944, value=2.0838}
         {arriveTangent=1.6993, interpMode=RCIM_Cubic, leaveTangent=1.6993, tangentMode=RCTM_User, time=1, value=3}
        ]}}})
    Index 0 Param 4=3
    Index 1 Param 1=&MergeHelper_FloatSwitch(0=1.2, 1=1.2, 2=2, 3=2, Index=&SpawnGroup(), Num=::NewEnumerator2)
    Index 1 Param 2=$Particles.Random
    Index 1 Param 3=0.7
    Index 1 Param 4=1
    Index 2 Param 1=&FloatFromCurve(FloatCurve=DataInterface() {propertyValues {Curve {keys [<at 0 1>,<at 1 0.3>]}}})
    Module.Index0Float0Write=true
    Module.Index0Float1Write=true
    Module.Index0Float2Write=true
    Module.Index0Float3Write=true
    Module.Index1Float0Write=true
    Module.Index1Float1Write=true
    Module.Index1Float2Write=true
    Module.Index1Float3Write=true
    Module.Index2Float0Write=true
    Module.Index2Float1Write=true
    Module.Index2Float2Write=true
    Module.Index2Float3Write=true
    Write Parameter Index 1=::NewEnumerator1
    Write Parameter Index 2=::NewEnumerator1
   )
   /SetVariables(Particles.Lightning glow color:LinearColor=&MakeLinearColorFromVectorAndFloat(
     "Vector (RGB)"=&MakeVectorFromLinearColorRGB(Color=&MultiplyLinearColor(
       Linear Color=<rgba 1.3 0.2025 0.0406 1>
       Scale Factor=&MakeVector4_FromFloat(Float=&FN_FrontEndFloatModifier(Gameplay=0.5, Front End=0.4))
      ))
     "Float (Alpha)"=&Multiply_Float(
      A=&FloatFromCurve(FloatCurve=DataInterface() {propertyValues {Curve {keys [<at 0 1 RCIM_Cubic>, {interpMode=RCIM_Cubic, leaveTangent=-3.6307, time=0.9, value=0}]}}})
      B=2
     )
    )) {isSetParameters=true}
   /SetVariables(Particles.Glow Dy02:Vector4f=&MakeVector4(
     X=&FloatFromCurve(FloatCurve=DataInterface() {propertyValues {Curve {keys [{tangentMode=RCTM_User, time=0, value=0.2}, {tangentMode=RCTM_User, time=0.5, value=1}]}}})
     Y=0.7
     Z=0
     W=3
    )) {isSetParameters=true}
   /SetVariables(Particles.Glow Dy03:Vector4f=&MakeVector4(X=&Multiply_Float(A=1, B=1), Y=$Particles.Random, Z=0.7, W=1)) {isSetParameters=true}
   &FN_Sprite_Attributes(
    Camera Offset=20
    Color=&MakeLinearColorFromVectorAndFloat(
     "Vector (RGB)"=&Multiply_Vector(A=&MakeVectorFromLinearColorRGB(Color=<rgba 1 0.141 0.0295 1>), B=<1 1 1>)
     "Float (Alpha)"=&Multiply_Float(
      A=&FloatFromCurve(FloatCurve=DataInterface() {propertyValues {Curve {keys [
           {interpMode=RCIM_Cubic, tangentMode=RCTM_User, time=0, value=0}
           {interpMode=RCIM_Cubic, tangentMode=RCTM_User, time=0.3, value=1}
           {tangentMode=RCTM_User, time=1, value=0}
          ]}}})
      B=1
     )
    )
    Create Camera Offset Attribute=true
    Create Color Attribute=true
    Create SpriteSize Attribute=true
    SpriteSize=&Multiply_Vector2D_ByFloat(
     Float=&FloatFromCurve(FloatCurve=DataInterface() {propertyValues {Curve {keys [
          {interpMode=RCIM_Cubic, tangentMode=RCTM_User, time=0, value=0}
          {interpMode=RCIM_Cubic, tangentMode=RCTM_User, time=0.3, value=1}
          {tangentMode=RCTM_User, time=1, value=0}
         ]}}})
     Vector2D=<40 40>
    )
    Write to Module Namespace=true
   )
   /SetVariables(Particles.DecalDynamic2:Vector4f=<0 0 1 1>) {isSetParameters=true}
  ]
  rendererClasses [&NiagaraMeshRendererProperties, &NiagaraSpriteRendererProperties]
  renderers [
   {rendererClass=&NiagaraMeshRendererProperties, rendererIndex=0}
   {rendererClass=&NiagaraMeshRendererProperties, rendererIndex=1}
   {rendererClass=&NiagaraSpriteRendererProperties, rendererIndex=2}
  ]
 }
}
summary {
 systemName=NS_Emote_SwordSplit_Anticipation_Lightning_02
 emitters [
  {emitterName=LightningArcs, rendererClasses [&NiagaraMeshRendererProperties, &NiagaraSpriteRendererProperties]}
  {
   emitterName=LightningStreaks
   rendererClasses [&NiagaraMeshRendererProperties, &NiagaraSpriteRendererProperties]
  }
 ]
 userVariables [
  User.bIsFrontEnd {type=NiagaraBool}
  User.bIsFrontEndPreview {type=NiagaraBool}
 ]
}
